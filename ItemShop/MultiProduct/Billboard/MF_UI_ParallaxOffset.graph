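# UI material function that produces mouse-driven parallax offsets, stored in the ItemShop multi-product billboard folder.
# MPC_MousePosition X/Y is normalized against a fixed 1920x1080, remapped to center-relative -1 to 1, then scaled by horizontal and vertical multiplier inputs.
# It outputs offset UVs and a pixel offset for vertex displacement, scaled by view size relative to 1080p.
# A View.CursorPosition branch with curve-atlas response shaping is unconnected to either output.
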MaterialFunction /ItemShop/MultiProduct/Billboard/MF_UI_ParallaxOffset
graph {nodes [
  n1 = Custom() {code="return float2(View.CursorPosition);", outputType=CMOT_Float2, showCode=true, inputs []}
  n2 = ViewSize()
  n3 = Multiply(A=Subtract(A=Divide(A=@n1, B=@n2), B=Constant2Vector() {g=0.5, r=0.5}) {constB=0.5}) {constB=2}
  n4 = ComponentMask(None=@n3) {r=true}
  n5 = CurveAtlasRowParameter(CurveTime=Abs(None=@n4)) {
   atlas=&/CA_MousePosition
   curve=&/Curve_MousePosition
   group=ParallaxOffset
   parameterName=MousePositionCurve
  }
  n6 = Constant() {r=1}
  n7 = If(A=@n4, "A > B"=@n6, "A == B"=@n6, "A < B"=Constant() {r=-1})
  n8 = ComponentMask(None=@n3) {g=true}
  n9 = CurveAtlasRowParameter(CurveTime=Abs(None=@n8)) {
   atlas=&CA_MousePosition
   curve=&Curve_MousePosition
   group=ParallaxOffset
   parameterName=MousePositionCurve
  }
  n10 = Constant() {r=1}
  n11 = If(A=@n8, "A > B"=@n10, "A == B"=@n10, "A < B"=Constant() {r=-1})
  AppendVector(A=Multiply(A=NamedRerouteUsage(), B=Multiply(A=@n5.R, B=@n7)), B=Multiply(A=Multiply(A=@n9.R, B=@n11), B=NamedRerouteUsage()))
  n12 = TextureCoordinate()
  n13 = AppendVector(
   A=CollectionParameter() {collection=&/MPC_MousePosition, parameterName=MousePositionX}
   B=CollectionParameter() {collection=&MPC_MousePosition, parameterName=MousePositionY}
  )
  n14 = Constant2Vector() {g=1080, r=1920}
  n15 = Multiply(A=Subtract(A=Divide(A=@n13, B=@n14), B=Constant2Vector() {g=0.5, r=0.5}) {constB=0.5}) {constB=2}
  n16 = AppendVector(
   A=Multiply(A=NamedRerouteUsage(), B=ComponentMask(None=@n15) {r=true})
   B=Multiply(A=ComponentMask(None=@n15) {g=true}, B=NamedRerouteUsage())
  )
  FunctionOutput(None=Subtract(A=FunctionInput(Preview=@n12) {bUsePreviewValueAsDefault=true, inputName=UV, inputType=FunctionInput_Vector2}, B=@n16)) {
   desc=This output is for when you want to use the parallax to offset UVs
   description=This output is for when you want to use the parallax to offset UVs
   outputName=Parallaxed UV
  }
  n17 = ViewSize()
  n18 = Constant2Vector() {g=1080, r=1920}
  FunctionOutput(None=Multiply(A=@n16, B=Divide(A=@n17, B=@n18))) {
   desc=This output is for when you want to use the parallax for vertex offset
   description=This output is for when you want to use the parallax for vertex offset
   outputName=Pixel Offset
  }
  NamedRerouteDeclaration(None=FunctionInput(Preview=Constant()) {bUsePreviewValueAsDefault=true, inputName=HorizontalOffsetMultiplier, inputType=FunctionInput_Scalar, sortPriority=2}) {name=offset X, nodeColor=<rgba 0.3828 1 0 1>}
  NamedRerouteDeclaration(None=FunctionInput(Preview=Constant()) {bUsePreviewValueAsDefault=true, inputName=VerticalOffsetMultiplier, inputType=FunctionInput_Scalar, sortPriority=3}) {name=offset Y, nodeColor=<rgba 0 1 0.6719 1>}
 ]}
