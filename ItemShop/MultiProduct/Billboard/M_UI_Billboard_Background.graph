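# Translucent UI material for the Item Shop MultiProduct billboard background, built around a texture parameter.
# UV scaling and parallax offsets drive two samples of that texture: one blended toward a darkened copy, the other modulated by blended noise textures.
# Named reroutes share the texture, an inverted box mask of the height-scaled UVs, and an eased radial vignette with a colored overlay.
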Material /ItemShop/MultiProduct/Billboard/M_UI_Billboard_Background
graph {
 properties {
  bUsedWithStaticMesh=false
  blendMode=BLEND_Translucent
  materialDomain=MD_UI
  pixelDepthOffsetMode=PDOM_Legacy
  usedShadingModels=Default Lit
 }
 nodes [
  n1 = MaterialFunctionCall(UV=TextureCoordinate(), TextureSize=Constant2Vector() {g=9, r=16}) {materialFunction=&MF_UI_UVScaling}
  n2 = Add(
   A=MaterialFunctionCall(UVs=@n1.ScaleY, Texture Scale=Constant() {r=1.01}) {materialFunction=&ScaleUVsByCenter}.UVs
   B=AppendVector(
    A=ScalarParameter() {group=ParallaxOffset, parameterName=TextureOffsetU}
    B=ScalarParameter() {group=ParallaxOffset, parameterName=TextureOffsetV}
   )
  )
  n3 = ScalarParameter() {group=ParallaxOffset, parameterName=HorizontalParallaxMultiplier, sortPriority=0}
  n4 = ScalarParameter() {group=ParallaxOffset, parameterName=VerticalParallaxMultiplier, sortPriority=0}
  n5 = TextureSample(
   UVs=LinearInterpolate(
    A=MaterialFunctionCall!(UV=@n2, HorizontalOffsetMultiplier=@n3, VerticalOffsetMultiplier=@n4) {materialFunction=&/MF_UI_ParallaxOffset}.Parallaxed UV
    B=@n1.Scaled To Fill
    Alpha=NamedRerouteUsage()
   )
   Tex=NamedRerouteUsage()
  ) {samplerSource=SSM_Clamp_WorldGroupSettings}
  LinearInterpolate(A=@n5.RGB, B=Multiply(A=@n5.RGB) {constB=0.25}, Alpha=NamedRerouteUsage())
  MaterialFunctionCall!(
   UV
   Corner=ScalarParameter() {defaultValue=1, parameterName=MaskCorner, sliderMax=3}
   Angle=ScalarParameter() {defaultValue=90, parameterName=MaskAngle, sliderMax=90}
   Flip=Constant() {r=1}
   Falloff=ScalarParameter() {defaultValue=1, parameterName=MaskSharpness, sliderMax=1}
  ) {materialFunction=/FortUILibrary/MaterialFunctions/Shapes/MF_UI_BoxAngle}
  n6 = MaterialFunctionCall(UVs=@n1.ScaleY, Texture Scale=AppendVector(A=Constant() {r=0.9}, B=Constant() {r=1.01})) {materialFunction=&ScaleUVsByCenter}
  n7 = AppendVector(
   A=ScalarParameter() {group=ParallaxOffset, parameterName=TextureOffsetU}
   B=ScalarParameter() {group=ParallaxOffset, parameterName=TextureOffsetV}
  )
  n8 = TextureSample(
   UVs=MaterialFunctionCall(
    UV=Add(A=@n6.UVs, B=@n7)
    HorizontalOffsetMultiplier=ScalarParameter() {group=ParallaxOffset, parameterName=HorizontalParallaxMultiplier, sortPriority=0}
    VerticalOffsetMultiplier=ScalarParameter() {group=ParallaxOffset, parameterName=VerticalParallaxMultiplier, sortPriority=0}
   ) {materialFunction=&MF_UI_ParallaxOffset}.Parallaxed UV
   Tex=NamedRerouteUsage()
  ) {samplerSource=SSM_Clamp_WorldGroupSettings}
  n9 = Multiply(A=TextureCoordinate(), B=AppendVector(A=Constant() {r=0.02}, B=Constant() {r=1.1}))
  n10 = Constant() {r=0.5}
  n11 = Constant() {r=1}
  n12 = Multiply(A=Add(A=TextureCoordinate(), B=AppendVector(A=@n10, B=@n11)), B=AppendVector(A=Constant() {r=0.05}, B=Constant() {r=1.35}))
  n13 = LinearInterpolate(
   A=Multiply(A=TextureSample(UVs=@n9) {samplerType=SAMPLERTYPE_LinearGrayscale, texture=/Game/Global/Textures/Default/Noise/LowResBlurredNoise_NoMips}.RGB) {constB=1.5}
   B=Multiply(A=TextureSample(UVs=@n12) {samplerType=SAMPLERTYPE_LinearGrayscale, texture=/Game/Global/Textures/Default/Noise/T_SwirlNoise_N_height}.RGB) {constB=1.2}
  )
  Multiply(A=Saturate(None=Multiply(A=@n8.RGB, B=@n13)), B=NamedRerouteUsage())
  NamedRerouteDeclaration(None=TextureObjectParameter() {parameterName=Texture}) {name=TextureRef, nodeColor=<rgba 1 0.6797 0 1>}
  n14 = Constant() {r=1}
  NamedRerouteDeclaration(None=OneMinus(None=MaterialFunctionCall(UVs=@n1.ScaleY, Size=@n14) {materialFunction=&MF_UI_Box}.Fill)) {name=BackgroundMask, nodeColor=<rgba 1 0.4688 0 1>}
  n15 = Constant()
  n16 = Constant() {r=-0.535}
  n17 = MaterialFunctionCall(UVs=Add(A=TextureCoordinate(), B=AppendVector(A=@n15, B=@n16)), Texture Scale=AppendVector(A=Constant() {r=1.45}, B=Constant() {r=3})) {materialFunction=&ScaleUVsByCenter}
  n18 = Constant() {r=0.75}
  NamedRerouteDeclaration(None=MaterialFunctionCall("Value (0-1)"=MaterialFunctionCall(UV=@n17.UVs, GradientEnd=@n18) {materialFunction=&MF_UI_RadialGradient}.Smooth) {materialFunction=&MF_UI_EaseCurves}.EaseInOutSine) {name=BackgroundVignette, nodeColor=<rgba 1 0 0.8203 1>}
  NamedRerouteDeclaration(None=LinearInterpolate(
    A=Constant3Vector() {constant=<rgba 0.0843 0.0845 0.1094 1>}
    B=Constant3Vector() {constant=<rgba 0.8325 0.8337 0.9688 1>}
    Alpha=NamedRerouteUsage()
   )) {name=ColoredVignetteOverlay, nodeColor=<rgba 0 0.6719 1 1>}
 ]
}
