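# Translucent UI material for the item shop multi-product billboard, named as its blurred background.
# It calls GetSlatePost0 to read the Slate post-process buffer, the likely source of the blur.
# A separate chain scales UVs for a 16:9 aspect, feeds a square-cornered MF_UI_Box and inverts its fill as a mask.
# The graph shows no links from either part to the material outputs.
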Material /ItemShop/MultiProduct/Billboard/M_UI_Billboard_Background_Blur
graph {
 properties {
  bAutomaticallySetUsageInEditor=false
  bUsedWithStaticMesh=false
  blendMode=BLEND_Translucent
  materialDomain=MD_UI
  pixelDepthOffsetMode=PDOM_Legacy
  usedShadingModels=Default Lit
 }
 nodes [
  MaterialFunctionCall!("UVs (Optional)") {materialFunction=/Engine/Functions/UserInterface/GetSlatePost0}
  n1 = TextureCoordinate()
  n2 = Constant2Vector() {g=9, r=16}
  OneMinus(None=MaterialFunctionCall(
    UVs=MaterialFunctionCall(UV=@n1, TextureSize=@n2) {materialFunction=&MF_UI_UVScaling}.ScaleY
    Size=Constant() {r=1}
    CornerRoundness=Constant()
   ) {materialFunction=&MF_UI_Box}.Fill)
 ]
}
