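# Translucent UI material for the foreground layer of an Item Shop multi-product billboard.
# A texture parameter is sampled with clamped, size-aware scaled UVs, using an adjustable horizontal alignment pivot and U/V offset.
# Its alpha is multiplied by a box fill and a rotatable, offsettable linear gradient that forms a wipe mask.
# A parallax function with horizontal and vertical multipliers adds a pixel offset to world position to shift the layer.
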Material /ItemShop/MultiProduct/Billboard/M_UI_Billboard_Foreground
graph {
 properties {
  bUsedWithStaticMesh=false
  blendMode=BLEND_Translucent
  materialDomain=MD_UI
  pixelDepthOffsetMode=PDOM_Legacy
  usedShadingModels=Default Lit
 }
 nodes [
  n1 = ScalarParameter() {group=ParallaxOffset, parameterName=HorizontalParallaxMultiplier, sortPriority=0}
  n2 = ScalarParameter() {group=ParallaxOffset, parameterName=VerticalParallaxMultiplier, sortPriority=0}
  Add(
   A=WorldPosition().XYZ
   B=AppendVector(
    A=MaterialFunctionCall!(UV, HorizontalOffsetMultiplier=@n1, VerticalOffsetMultiplier=@n2) {materialFunction=&/MF_UI_ParallaxOffset}.Pixel Offset
    B=Constant()
   )
  )
  n3 = NamedRerouteUsage()
  n4 = AppendVector(A=ScalarParameter() {parameterName=TextureOffsetU}, B=ScalarParameter() {parameterName=TextureOffsetV})
  n5 = NamedRerouteUsage()
  n6 = Constant() {r=1}
  n7 = Constant() {r=0.45}
  n8 = Constant() {r=0.5}
  n9 = ScalarParameter() {defaultValue=-0.24, parameterName=WipeDirection}
  n10 = ScalarParameter() {parameterName=WipeOffset}
  Multiply(
   A=TextureSample(UVs=Add(A=@n3, B=@n4), Tex=NamedRerouteUsage()) {samplerSource=SSM_Clamp_WorldGroupSettings}.A
   B=Multiply(
    A=MaterialFunctionCall(UVs=@n5, Size=@n6) {materialFunction=&MF_UI_Box}.Fill
    B=MaterialFunctionCall(GradientStart=@n7, GradientEnd=@n8, RotationAngle=@n9, GradientOffset=@n10) {materialFunction=&MF_UI_LinearGradient}.Result
   )
  )
  NamedRerouteDeclaration(None=TextureObjectParameter() {parameterName=Texture}) {name=TextureRef, nodeColor=<rgba 1 0.6797 0 1>}
  n11 = NamedRerouteUsage()
  n12 = ScalarParameter() {defaultValue=0.5, desc="0 is Left aligned, 0.5 is Center aligned, and 1 is Right aligned.", parameterName=TextureAlignment}
  n13 = Constant()
  NamedRerouteDeclaration(None=MaterialFunctionCall(UV=TextureCoordinate(), TextureSize=TextureProperty(Tex=@n11), ScalePivotPoint=AppendVector(A=@n12, B=@n13)) {materialFunction=&MF_UI_UVScaling}.ScaleY) {name=ScaledUV, nodeColor=<rgba 0 0.8125 1 1>}
 ]
}
