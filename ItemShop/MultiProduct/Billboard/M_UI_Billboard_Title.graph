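# UI material for the item shop multi-product billboard title, producing a tinted two-color gradient.
# MF_UI_LinearGradient, fed by texture coordinates, drives a lerp between two exposed color parameters, defaulting to deep blue and pale cyan.
# The gradient is multiplied by vertex color, so the widget's tint modulates the result in this translucent material.
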Material /ItemShop/MultiProduct/Billboard/M_UI_Billboard_Title
graph {
 properties {
  bUsedWithStaticMesh=false
  blendMode=BLEND_Translucent
  materialDomain=MD_UI
  pixelDepthOffsetMode=PDOM_Legacy
  usedShadingModels=Default Lit
 }
 nodes [
  n1 = TextureCoordinate()
  Multiply(
   A=VertexColor()
   B=LinearInterpolate(
    A=VectorParameter() {parameterName=Gradient Color 1, defaultValue=<rgba 0 0.2831 0.4452 1>}.RGB
    B=VectorParameter() {parameterName=Gradient Color 2, defaultValue=<rgba 0.521 0.8963 1 1>}.RGB
    Alpha=MaterialFunctionCall(UVs=@n1) {materialFunction=&MF_UI_LinearGradient}.Result
   )
  )
 ]
}
