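# Item Shop billboard widget that shows an offer group's background and foreground art with a blur layer between them.
# String metadata selects both images and the foreground alignment; backgrounds load from disk, falling back to URL.
# Activation and deactivation play material wipe and slide animations chosen by last scroll direction, then fire completion delegates.
# Brushes are cleared on deactivation completion and destruct, releasing optionally managed textures.
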WidgetBlueprint /ItemShop/MultiProduct/Billboard/WBP_Billboard_Background
animations [
 EaseIn {
  binding_count=2
  length_seconds=0.8
  object=&self:EaseIn
  sequence {bindings [
    LazyImage_Foreground {
     id=#1
     possessed_class=FortLazyImage
     tracks [
      {
       type=MovieSceneWidgetMaterialTrack
       sections [{
         type=MovieSceneParameterSection
         end_frame=15
         start_frame=0
         channels [
          WipeOffset {type=MovieSceneScriptingFloatChannel, keys [[4,0.55],[10,-0.55]]}
          WipeDirection {type=MovieSceneScriptingFloatChannel, const [-0.24,1]}
         ]
        }]
      }
      {
       type=MovieSceneFloatTrack
       display_name=Render Opacity
       sections [{
         type=MovieSceneFloatSection
         end_frame=15
         start_frame=0
         channels [None {type=MovieSceneScriptingFloatChannel, keys [[0,0],[4,0],[10,1],[12,1]]}]
        }]
      }
      {
       type=MovieScene2DTransformTrack
       sections [{
         type=MovieScene2DTransformSection
         end_frame=16
         start_frame=0
         channels [
          Translation.X {type=MovieSceneScriptingFloatChannel, const [0,1]}
          Translation.Y {type=MovieSceneScriptingFloatChannel, keys [[4,102],[12,48]]}
          Scale.X {type=MovieSceneScriptingFloatChannel, keys [[4,0.97],[16,1]]}
          Scale.Y {type=MovieSceneScriptingFloatChannel, keys [[4,0.97],[16,1]]}
         ]
        }]
      }
     ]
    }
    LazyImage_Background {
     id=#2
     possessed_class=FortLazyImage
     tracks [{
       type=MovieSceneFloatTrack
       display_name=Render Opacity
       sections [{type=MovieSceneFloatSection, end_frame=5, start_frame=0, channels [None {type=MovieSceneScriptingFloatChannel, keys [[0,0],[5,1]]}]}]
      }]
    }
   ]}
 }
 EaseIn_ScrolledDown {
  binding_count=2
  length_seconds=0.95
  object=&self:EaseIn_ScrolledDown
  sequence {bindings [
    LazyImage_Foreground {
     id=#1
     possessed_class=FortLazyImage
     tracks [
      {
       type=MovieSceneFloatTrack
       display_name=Render Opacity
       sections [{
         type=MovieSceneFloatSection
         end_frame=15
         start_frame=0
         channels [None {type=MovieSceneScriptingFloatChannel, keys [[0,0],[7,0],[13,1],[15,1]]}]
        }]
      }
      {
       type=MovieScene2DTransformTrack
       sections [{
         type=MovieScene2DTransformSection
         end_frame=19
         start_frame=0
         channels [
          Translation.X {type=MovieSceneScriptingFloatChannel, const [0,1]}
          Translation.Y {type=MovieSceneScriptingFloatChannel, keys [[7,102],[15,48]]}
          Scale.X {type=MovieSceneScriptingFloatChannel, keys [[7,0.97],[19,1]]}
          Scale.Y {type=MovieSceneScriptingFloatChannel, keys [[7,0.97],[19,1]]}
         ]
        }]
      }
      {
       type=MovieSceneWidgetMaterialTrack
       sections [{
         type=MovieSceneParameterSection
         end_frame=15
         start_frame=0
         channels [
          WipeOffset {type=MovieSceneScriptingFloatChannel, keys [[7,0.55],[13,-0.55]]}
          WipeDirection {type=MovieSceneScriptingFloatChannel, const [-0.24,1]}
         ]
        }]
      }
     ]
    }
    LazyImage_Background {
     id=#2
     possessed_class=FortLazyImage
     tracks [
      {
       type=MovieScene2DTransformTrack
       sections [{
         type=MovieScene2DTransformSection
         end_frame=15
         start_frame=0
         channels [Translation.Y {type=MovieSceneScriptingFloatChannel, keys [[0,200],[10,0]]}]
        }]
      }
      {
       type=MovieSceneWidgetMaterialTrack
       sections [{
         type=MovieSceneParameterSection
         end_frame=15
         start_frame=0
         channels [
          MaskAngle {type=MovieSceneScriptingFloatChannel, keys [[0,85],[10,90]]}
          MaskSharpness {type=MovieSceneScriptingFloatChannel, keys [[0,0.25],[10,1]]}
          MaskCorner {type=MovieSceneScriptingFloatChannel, const [1,1]}
         ]
        }]
      }
     ]
    }
   ]}
 }
 EaseIn_ScrolledUp {
  binding_count=2
  length_seconds=0.95
  object=&self:EaseIn_ScrolledUp
  sequence {bindings [
    LazyImage_Foreground {
     id=#1
     possessed_class=FortLazyImage
     tracks [
      {
       type=MovieSceneFloatTrack
       display_name=Render Opacity
       sections [{
         type=MovieSceneFloatSection
         end_frame=15
         start_frame=0
         channels [None {type=MovieSceneScriptingFloatChannel, keys [[0,0],[7,0],[13,1],[15,1]]}]
        }]
      }
      {
       type=MovieScene2DTransformTrack
       sections [{
         type=MovieScene2DTransformSection
         end_frame=19
         start_frame=0
         channels [
          Translation.X {type=MovieSceneScriptingFloatChannel, const [0,1]}
          Translation.Y {type=MovieSceneScriptingFloatChannel, keys [[7,-54],[15,48]]}
          Scale.X {type=MovieSceneScriptingFloatChannel, keys [[7,0.97],[19,1]]}
          Scale.Y {type=MovieSceneScriptingFloatChannel, keys [[7,0.97],[19,1]]}
         ]
        }]
      }
      {
       type=MovieSceneWidgetMaterialTrack
       sections [{
         type=MovieSceneParameterSection
         end_frame=15
         start_frame=0
         channels [
          WipeOffset {type=MovieSceneScriptingFloatChannel, keys [[7,0.55],[13,-0.55]]}
          WipeDirection {type=MovieSceneScriptingFloatChannel, const [0.26,1]}
         ]
        }]
      }
     ]
    }
    LazyImage_Background {
     id=#2
     possessed_class=FortLazyImage
     tracks [
      {
       type=MovieSceneFloatTrack
       display_name=Render Opacity
       sections [{type=MovieSceneFloatSection, end_frame=8, start_frame=0, channels [None {type=MovieSceneScriptingFloatChannel, const [1,2]}]}]
      }
      {
       type=MovieSceneWidgetMaterialTrack
       sections [{
         type=MovieSceneParameterSection
         end_frame=15
         start_frame=0
         channels [
          MaskAngle {type=MovieSceneScriptingFloatChannel, keys [[0,85],[10,90]]}
          MaskSharpness {type=MovieSceneScriptingFloatChannel, keys [[0,0.25],[10,1]]}
          MaskCorner {type=MovieSceneScriptingFloatChannel, const [3,1]}
         ]
        }]
      }
      {
       type=MovieScene2DTransformTrack
       sections [{
         type=MovieScene2DTransformSection
         end_frame=15
         start_frame=0
         channels [Translation.Y {type=MovieSceneScriptingFloatChannel, keys [[0,-200],[10,0]]}]
        }]
      }
     ]
    }
   ]}
 }
 EaseOut {
  binding_count=2
  length_seconds=0.4
  object=&self:EaseOut
  sequence {bindings [
    LazyImage_Foreground {
     id=#1
     possessed_class=FortLazyImage
     tracks [
      {
       type=MovieSceneFloatTrack
       display_name=Render Opacity
       sections [{type=MovieSceneFloatSection, end_frame=5, start_frame=0, channels [None {type=MovieSceneScriptingFloatChannel, keys [[0,1],[5,0]]}]}]
      }
      {
       type=MovieSceneWidgetMaterialTrack
       sections [{
         type=MovieSceneParameterSection
         end_frame=8
         start_frame=0
         channels [
          WipeOffset {type=MovieSceneScriptingFloatChannel, keys [[0,-0.5],[8,0.5]]}
          WipeDirection {type=MovieSceneScriptingFloatChannel, const [-0.24,1]}
         ]
        }]
      }
      {
       type=MovieScene2DTransformTrack
       sections [{
         type=MovieScene2DTransformSection
         end_frame=8
         start_frame=0
         channels [
          Translation.X {type=MovieSceneScriptingFloatChannel, const [0,1]}
          Translation.Y {type=MovieSceneScriptingFloatChannel, keys [[0,48],[8,1024]]}
         ]
        }]
      }
     ]
    }
    LazyImage_Background {
     id=#2
     possessed_class=FortLazyImage
     tracks [{
       type=MovieSceneFloatTrack
       display_name=Render Opacity
       sections [{type=MovieSceneFloatSection, end_frame=5, start_frame=0, channels [None {type=MovieSceneScriptingFloatChannel, keys [[0,1],[5,0]]}]}]
      }]
    }
   ]}
 }
 EaseOut_ScrolledDown {
  binding_count=2
  length_seconds=0.3
  object=&self:EaseOut_ScrolledDown
  sequence {bindings [
    LazyImage_Foreground {
     id=#1
     possessed_class=FortLazyImage
     tracks [
      {
       type=MovieSceneFloatTrack
       display_name=Render Opacity
       sections [{type=MovieSceneFloatSection, end_frame=3, start_frame=0, channels [None {type=MovieSceneScriptingFloatChannel, keys [[0,1],[3.75,0]]}]}]
      }
      {
       type=MovieScene2DTransformTrack
       sections [{
         type=MovieScene2DTransformSection
         end_frame=6
         start_frame=0
         channels [
          Translation.X {type=MovieSceneScriptingFloatChannel, const [0,1]}
          Translation.Y {type=MovieSceneScriptingFloatChannel, keys [[0,48],[6,-1024]]}
         ]
        }]
      }
      {
       type=MovieSceneWidgetMaterialTrack
       sections [{
         type=MovieSceneParameterSection
         end_frame=6
         start_frame=0
         channels [
          WipeOffset {type=MovieSceneScriptingFloatChannel, keys [[0,-0.5],[6,0.5]]}
          WipeDirection {type=MovieSceneScriptingFloatChannel, const [0.26,1]}
         ]
        }]
      }
     ]
    }
    LazyImage_Background {
     id=#2
     possessed_class=FortLazyImage
     tracks [
      {
       type=MovieScene2DTransformTrack
       sections [{
         type=MovieScene2DTransformSection
         end_frame=6
         start_frame=0
         channels [Translation.Y {type=MovieSceneScriptingFloatChannel, keys [[0,0],[6,-1080]]}]
        }]
      }
      {
       type=MovieSceneWidgetMaterialTrack
       sections [{
         type=MovieSceneParameterSection
         end_frame=6
         start_frame=0
         channels [
          MaskAngle {type=MovieSceneScriptingFloatChannel, keys [[0,90],[5.25,85]]}
          MaskSharpness {type=MovieSceneScriptingFloatChannel, keys [[0,1],[5.25,0.25]]}
          MaskCorner {type=MovieSceneScriptingFloatChannel, const [3,1]}
         ]
        }]
      }
     ]
    }
   ]}
 }
 EaseOut_ScrolledUp {
  binding_count=2
  length_seconds=0.3
  object=&self:EaseOut_ScrolledUp
  sequence {bindings [
    LazyImage_Foreground {
     id=#1
     possessed_class=FortLazyImage
     tracks [
      {
       type=MovieScene2DTransformTrack
       sections [{
         type=MovieScene2DTransformSection
         end_frame=6
         start_frame=0
         channels [
          Translation.X {type=MovieSceneScriptingFloatChannel, const [0,1]}
          Translation.Y {type=MovieSceneScriptingFloatChannel, keys [[0,48],[6,1024]]}
         ]
        }]
      }
      {
       type=MovieSceneWidgetMaterialTrack
       sections [{
         type=MovieSceneParameterSection
         end_frame=6
         start_frame=0
         channels [
          WipeOffset {type=MovieSceneScriptingFloatChannel, keys [[0,-0.5],[6,0.5]]}
          WipeDirection {type=MovieSceneScriptingFloatChannel, const [-0.24,1]}
         ]
        }]
      }
     ]
    }
    LazyImage_Background {
     id=#2
     possessed_class=FortLazyImage
     tracks [
      {
       type=MovieScene2DTransformTrack
       sections [{
         type=MovieScene2DTransformSection
         end_frame=6
         start_frame=0
         channels [Translation.Y {type=MovieSceneScriptingFloatChannel, keys [[0,0],[6,1080]]}]
        }]
      }
      {
       type=MovieSceneWidgetMaterialTrack
       sections [{
         type=MovieSceneParameterSection
         end_frame=6
         start_frame=0
         channels [
          MaskAngle {type=MovieSceneScriptingFloatChannel, keys [[0,90],[5.25,85]]}
          MaskSharpness {type=MovieSceneScriptingFloatChannel, keys [[0,1],[5.25,0.25]]}
         ]
        }]
      }
     ]
    }
   ]}
 }
]
graphs [
 EventGraph {nodes [
   n1 = OnBackgroundLoadedFromDisk!() -> (OutputDelegate, then, LoadedBackground:Object Reference) {is=title K2Node_CustomEvent}
   > n2 = Cast To MaterialInterface!(execute, Object=@n1.LoadedBackground) -> (then, CastFailed, AsMaterial Interface:Material Interface Object Reference) {is=title K2Node_DynamicCast}
   > n3 = SetBrushFromMaterial(self=Get LazyImage_Background() -> (LazyImage_Background:Fort Lazy Image Object Reference), Material=@n2)
   TryAddToManagedTextures!(execute=[@n3.then, @n5.then], self, LoadedAsset:Object Reference=@n1.LoadedBackground) -> (then)
   n4 = Cast To Texture2D(execute=@n2.CastFailed, Object=@n1.LoadedBackground)
   > n5 = SetBrushFromTexture!(execute, self:Image Object Reference=Get LazyImage_Background(), Texture:Texture 2D Object Reference=@n4, bMatchSize:bool=false) -> (then)
   OnActivated() {is=title K2Node_CustomEvent}
   > Set IsActive(IsActive:bool=true) -> (Output_Get:bool)
   > UnbindAllAnimEvents()
   > StopAnimationsAndLatentActions(self:User Widget Object Reference)
   > UpdateBrushes()
   > ResetAnimationState()
   > n6 = UpdateVisibility()
   n7 = OnActivationAnimFinished() {is=title K2Node_CustomEvent}
   n8 = Select!(
    NoScroll:Widget Animation Object Reference=Get EaseIn() -> (EaseIn:Widget Animation Object Reference)
    ScrolledDown:Widget Animation Object Reference=Get EaseIn_ScrolledDown() -> (EaseIn_ScrolledDown:Widget Animation Object Reference)
    ScrolledUp:Widget Animation Object Reference=Get EaseIn_ScrolledUp() -> (EaseIn_ScrolledUp:Widget Animation Object Reference)
    Index:EFortItemShopScrollDirection Enum=Get LastScrolledDirection!(self:Fort Item Shop Selection VM Object Reference=Get FortItemShopSelectionVM() -> (FortItemShopSelectionVM:Fort Item Shop Selection VM Object Reference)) -> (LastScrolledDirection:EFortItemShopScrollDirection Enum=NoScroll)
   ) -> (ReturnValue:Widget Animation Object Reference)
   BindToAnimationFinished(execute=@n6.then, Animation=@n8, Delegate=@n7)
   > PlayAnimation(InAnimation=@n8)
   > Call On Activation Started!(execute:Exec, self:WBP Billboard Background Object Reference, Widget:WBP Billboard Background Object Reference=Self-Reference()) -> (then:Exec) {is=title K2Node_CallDelegate}
   OnDeactivated() {is=title K2Node_CustomEvent}
   > Set IsActive(IsActive=false)
   > UnbindAllAnimEvents()
   > StopAnimationsAndLatentActions()
   > n9 = ResetAnimationState()
   n10 = OnDeactivateAnimFinished() {is=title K2Node_CustomEvent}
   n11 = Select(
    NoScroll=Get EaseOut() -> (EaseOut:Widget Animation Object Reference)
    ScrolledDown=Get EaseOut_ScrolledDown() -> (EaseOut_ScrolledDown:Widget Animation Object Reference)
    ScrolledUp=Get EaseOut_ScrolledUp() -> (EaseOut_ScrolledUp:Widget Animation Object Reference)
    Index=Get LastScrolledDirection(self=Get FortItemShopSelectionVM()) -> (LastScrolledDirection=NoScroll)
   )
   BindToAnimationFinished(execute=@n9.then, Animation=@n11, Delegate=@n10)
   > PlayAnimation(InAnimation=@n11)
   n12 = OnTextureLoadedFromUrl!() -> (OutputDelegate, then, Texture:Texture 2D Object Reference) {is=title K2Node_CustomEvent}
   > TryAddToManagedTextures(LoadedAsset=@n12.Texture)
   UnbindAllAnimEvents() {is=title K2Node_CustomEvent}
   > UnbindAllFromAnimationFinished!(execute, self:User Widget Object Reference, Animation:Widget Animation Object Reference=Get EaseOut()) -> (then)
   > UnbindAllFromAnimationFinished(Animation=Get EaseOut_ScrolledUp())
   > UnbindAllFromAnimationFinished(Animation=Get EaseOut_ScrolledDown())
   > UnbindAllFromAnimationFinished(Animation=Get EaseIn())
   > UnbindAllFromAnimationFinished(Animation=Get EaseIn_ScrolledUp())
   > UnbindAllFromAnimationFinished(Animation=Get EaseIn_ScrolledDown())
   n13 = Event PreConstruct()
   > n14 = Branch(Condition=@n13.IsDesignTime)
   > SetDesignTimePreview()
   Set IsActive(execute=@n14.else, IsActive=false)
   > UpdateVisibility()
   > Branch(Condition=IsMobileGame())
   n15 = GetViewportSize!() -> (ReturnValue_X:double, ReturnValue_Y:double)
   > n16 = Branch(Condition="float >= float"(A=float / float(A=@n15, B=@n15.ReturnValue_Y), B=1.34))
   SetSize!(execute=@n16.else, self:Spacer Object Reference=Get Spacer() -> (Spacer:Spacer Object Reference), InSize:Vector 2D Structure=(X=1920,Y=1440)) -> (then)
   Event Destruct()
   > ReleaseBrushes()
   Call On Activation Complete!(execute:Exec=@n20.then, self:WBP Billboard Background Object Reference, Widget:WBP Billboard Background Object Reference=Self-Reference()) -> (then:Exec) {is=title K2Node_CallDelegate}
   Call On Deactivation Complete!(execute:Exec=@n17.then, self:WBP Billboard Background Object Reference, Widget:WBP Billboard Background Object Reference=Self-Reference()) -> (then:Exec) {is=title K2Node_CallDelegate}
   n17 = ReleaseBrushes(execute=@n18.then)
   n18 = UpdateVisibility(execute=@n19.else)
   n19 = Branch(execute=@n10.then, Condition=Get IsActive() -> (IsActive:bool))
   n20 = Branch(execute=@n7.then, Condition=Get IsActive())
  ]}
 GetForegroundURL {nodes [
   n1 = GetForegroundURL!() -> (then, StringMetadata:Map of Names to Strings, Context:Gameplay Tag Structure) {is=title K2Node_FunctionEntry}
   > n2 = FindItemShopConfig!(execute, ProductTag:Gameplay Tag Structure=@n1.Context) -> (then, ReturnValue:Fort Product Item Shop Config Object Reference)
   > n3 = Is Valid(InputObject=@n2)
   n4 = Find(TargetMap:Map of Names to Strings=Get StringMetadata() -> (StringMetadata:Map of Names to Strings), Key:"Name (by ref)"=Foreground)
   Return Node!(
    execute=[@n6.else, @n3.Is Not Valid]
    ForegroundURL:String=Select!(Option 0:String, Option 1:String=@n4, Index:bool=@n4.ReturnValue) -> (ReturnValue:String)
   )
   n5 = Find(
    TargetMap=Get StringMetadata()
    Key=Get ForegroundMetadataName!(self:Fort Product Item Shop Config Object Reference=@n2) -> (ForegroundMetadataName:Name=None)
   )
   > n6 = Branch(Condition=@n5.ReturnValue)
   > Return Node(ForegroundURL=@n5)
  ]}
 OnActivationComplete {nodes [OnActivationComplete!() -> (then, Widget:WBP Billboard Background Object Reference) {is=title K2Node_FunctionEntry}]}
 OnActivationStarted {nodes [OnActivationStarted!() -> (then, Widget:WBP Billboard Background Object Reference) {is=title K2Node_FunctionEntry}]}
 OnDeactivationComplete {nodes [OnDeactivationComplete!() -> (then, Widget:WBP Billboard Background Object Reference) {is=title K2Node_FunctionEntry}]}
 ReleaseBrushes {nodes [
   ReleaseBrushes() {is=title K2Node_FunctionEntry}
   > SetBrushFromTexture(self=Get LazyImage_Background(), Texture=Get EmptyBrush() -> (EmptyBrush:Texture 2D Object Reference), bMatchSize=false)
   > SetBrushFromTexture(self=Get LazyImage_Foreground() -> (LazyImage_Foreground:Fort Lazy Image Object Reference), Texture=Get EmptyBrush(), bMatchSize=false)
   > ReleaseResource(self:Fort Managed Texture Object Object Reference=Get TextureManager() -> (TextureManager:Fort Managed Texture Object Object Reference))
  ]}
 ResetAnimationState {nodes [
   ResetAnimationState() {is=title K2Node_FunctionEntry}
   > n1 = Sequence()
   n2 = Get LazyImage_Background()
   > Set RenderOpacity!(execute, RenderOpacity:float=1, self:Widget Object Reference=@n2) -> (then, Output_Get:float)
   > Set RenderTransform!(execute, RenderTransform:Widget Transform Structure, self:Widget Object Reference=@n2) -> (then, Output_Get:Widget Transform Structure)
   > n3 = GetDynamicMaterial(self=@n2)
   > SetScalarParameterValue(self=@n3, ParameterName=MaskAngle, Value=90)
   > SetScalarParameterValue(self=@n3, ParameterName=MaskCorner, Value=1)
   > SetScalarParameterValue(self=@n3, ParameterName=MaskSharpness, Value=1)
   n4 = Get LazyImage_Foreground()
   Set RenderOpacity(execute=@n1.then_1, RenderOpacity=1, self=@n4)
   > Set RenderTransform(self=@n4)
   > n5 = GetDynamicMaterial(self=@n4)
   > SetScalarParameterValue(self=@n5, ParameterName=WipeDirection, Value=-24)
   > SetScalarParameterValue(self=@n5, ParameterName=WipeOffset, Value=-0.55)
  ]}
 SetDesignTimePreview {nodes [
   SetDesignTimePreview() {is=title K2Node_FunctionEntry}
   > SetBrushFromLazyTexture!(
    execute
    self:Common Lazy Image Object Reference=Get LazyImage_Background()
    LazyTexture:"Texture 2D Soft Object Reference (by ref)"=Get PreviewBackgroundImage() -> (PreviewBackgroundImage:Texture 2D Soft Object Reference)
    bMatchSize:bool=false
   ) -> (then)
   > SetBrushFromLazyTexture(
    self=Get LazyImage_Foreground()
    LazyTexture=Get PreviewForegroundImage() -> (PreviewForegroundImage:Texture 2D Soft Object Reference)
    bMatchSize=false
   )
  ]}
 SetOfferGroup {nodes [
   n1 = SetOfferGroup!() -> (then, OfferGroupVM:Fort Item Shop Offer Group VM Object Reference) {is=title K2Node_FunctionEntry}
   > SetFortItemShopOfferGroupVM!(execute, self, Viewmodel:Fort Item Shop Offer Group VM Object Reference=@n1) -> (then)
  ]}
 TryAddToManagedTextures {nodes [
   TryAddToManagedTextures!() -> (then, LoadedAsset:Object Reference) {is=title K2Node_FunctionEntry}
   > Branch(Condition=ShouldCleanupBillboardTextures!(self:Fort Item Shop Selection VM Object Reference=Get FortItemShopSelectionVM()) -> (ReturnValue:bool))
   > n1 = Branch(Condition=IsNullItem!(ItemObject:Object Reference=Get TextureManager()) -> (Is Null Item:bool))
   > n2 = Construct Object from Class(Class=/Script/FortniteGame.FortManagedTextureObject) -> (ReturnValue:Fort Managed Texture Object Object Reference)
   > n3 = Set TextureManager(TextureManager:Fort Managed Texture Object Object Reference=@n2) -> (Output_Get:Fort Managed Texture Object Object Reference)
   n4 = Get LoadedAsset() -> (LoadedAsset:Object Reference)
   n5 = Cast To MaterialInterface(execute=[@n3.then, @n1.else], Object=@n4)
   n6 = Get TextureManager()
   > AddMaterial!(execute, self:Fort Managed Texture Object Object Reference=@n6, Material:Material Interface Object Reference=@n5) -> (then)
   n7 = Cast To Texture2D(execute=@n5.CastFailed, Object=@n4)
   > AddTexture!(execute, self:Fort Managed Texture Object Object Reference=@n6, Texture:Texture Object Reference=@n7) -> (then)
  ]}
 UpdateAlignment {nodes [
   n1 = UpdateAlignment!() -> (then, StringMetadata:Map of Names to Strings) {is=title K2Node_FunctionEntry}
   > n2 = GetAlignmentFromStringMetadata!(execute, StringMetadata:Map of Names to Strings=@n1) -> (then, Alignment:EHorizontalAlignment Enum=HAlign_Fill, Found:bool)
   > n3 = GetDynamicMaterial(self=Get LazyImage_Foreground())
   n4 = Select!(
    HAlign_Fill:EHorizontalAlignment Enum=HAlign_Right
    HAlign_Left:EHorizontalAlignment Enum=HAlign_Right
    HAlign_Center:EHorizontalAlignment Enum=HAlign_Center
    HAlign_Right:EHorizontalAlignment Enum=HAlign_Left
    Index:EHorizontalAlignment Enum=@n2
   ) -> (ReturnValue:EHorizontalAlignment Enum)
   > SetScalarParameterValue(
    self=@n3
    ParameterName=TextureAlignment
    Value=Select(HAlign_Fill:float=0, HAlign_Left:float=0, HAlign_Center:float=0.5, HAlign_Right:float=1, Index=@n4) -> (ReturnValue:float)
   )
   n5 = Get ForegroundLeftPadding() -> (ForegroundLeftPadding:Margin Structure)
   > Set Padding!(
    execute
    Padding:Margin Structure=Select(
     HAlign_Fill:Margin Structure=@n5
     HAlign_Left:Margin Structure=@n5
     HAlign_Center:Margin Structure=Get ForegroundCenterPadding() -> (ForegroundCenterPadding:Margin Structure)
     HAlign_Right:Margin Structure=Get ForegroundRightPadding() -> (ForegroundRightPadding:Margin Structure)
     Index=@n4
    ) -> (ReturnValue:Margin Structure)
    self:Overlay Slot Object Reference=SlotAsOverlaySlot(Widget=Get LazyImage_Foreground())
   ) -> (then, Output_Get:Margin Structure)
  ]}
 UpdateBackground {nodes [
   n1 = UpdateBackground!() -> (then, StringMetadata:Map of Names to Strings) {is=title K2Node_FunctionEntry}
   n2 = Find(TargetMap=@n1, Key=Background)
   > Set BackgroundString(BackgroundString:String=@n2) -> (Output_Get:String)
   > Branch(Condition=@n2.ReturnValue)
   > SetBrushFromTexture(self=Get LazyImage_Background(), Texture=/FortUILibrary/Input/PeripheralGuitar/Source/Empty1x1, bMatchSize=false)
   > n3 = LoadBackground!(
    execute
    self:Fort Item Shop VM Object Reference=Get FortItemShopVM() -> (FortItemShopVM:Fort Item Shop VM Object Reference)
    Background:"Fort Item Shop CMSBackground Structure (by ref)"=Make Fort Item Shop CMSBackground!(CustomTexture:String, CookedAssetKey:String=Get BackgroundString() -> (BackgroundString:String), Type:EFortItemShopBackgroundType Enum=Light) -> (FortItemShopCMSBackground:Fort Item Shop CMSBackground Structure) {is=title K2Node_MakeStruct}
    OnLoadedCallback:"Delegate (by ref)"=Create Event()
   ) -> (then, ReturnValue:bool)
   > n4 = Branch(Condition=@n3)
   SetBrushFromMaterial(execute=@n4.else, self=Get LazyImage_Background(), Material=&/MI_UI_Billboard_Background)
   > SetMaterialTextureParamName!(execute, self:Common Lazy Image Object Reference=Get LazyImage_Background(), TextureParamName:Name=Texture) -> (then)
   > SetBrushFromURLWithCallback!(
    execute
    self:Fort Lazy Image Object Reference=Get LazyImage_Background()
    CMSImageURL:String=Get BackgroundString()
    bMatchSize:bool=false
    OnLoadedCallback:"Delegate (by ref)"=Create Event()
    Delay:float=0
    DownloadPriority:ECMSDownloadPriority Enum=Normal
   ) -> (then)
  ]}
 UpdateBrushes {nodes [
   UpdateBrushes() {is=title K2Node_FunctionEntry}
   > n1 = Get!(execute:Exec) -> (then:Exec, else:Exec, FortItemShopOfferGroupVM:Fort Item Shop Offer Group VM Object Reference)
   n2 = Get StringMetadata!(self:Fort Item Shop Offer Group VM Object Reference=@n1) -> (StringMetadata)
   > UpdateBackground!(execute, self, StringMetadata:Map of Names to Strings=@n2) -> (then) {is=title K2Node_CallFunction}
   n3 = Get FortItemShopVM()
   > UpdateForeground!(
    execute
    self
    CMSImageURL:"String (by ref)"=GetForegroundURL!(
     self
     StringMetadata:Map of Names to Strings=@n2
     Context:Gameplay Tag Structure=GetItemShopContext!(self:Fort Item Shop VM Object Reference=@n3) -> (ReturnValue:Gameplay Tag Structure)
    ) -> (ForegroundURL:String) {is=title K2Node_CallFunction}
   ) -> (then)
  ]}
 UpdateForeground {nodes [
   n1 = UpdateForeground!() -> (then, CMSImageURL:"String (by ref)") {is=title K2Node_FunctionEntry}
   > n2 = Branch(Condition=IsEmpty(InString=@n1))
   > SetBrushFromTexture(self=Get LazyImage_Foreground(), Texture=Get EmptyBrush(), bMatchSize=false)
   n3 = Get LazyImage_Foreground()
   > Set Visibility(Visibility=Collapsed, self=@n3)
   SetBrushFromURLWithCallback(
    execute=@n2.else
    self=Get LazyImage_Foreground()
    CMSImageURL=Get CMSImageURL() -> (CMSImageURL:"String (by ref)")
    bMatchSize=false
    OnLoadedCallback=Create Event()
    Delay=0
    DownloadPriority=Normal
   )
   > Set Visibility(Visibility=SelfHitTestInvisible, self=@n3)
  ]}
 UpdateVisibility {nodes [
   UpdateVisibility() {is=title K2Node_FunctionEntry}
   > n1 = Branch(Condition=Get IsActive())
   > Set Visibility!(execute, Visibility=SelfHitTestInvisible) -> (then, Output_Get)
   Set Visibility(execute=@n1.else, Visibility=Collapsed)
  ]}
 VM String Metadata Updated {nodes [
   VM String Metadata Updated!() -> (then, StringMetadata:Map of Names to Strings) {is=title K2Node_FunctionEntry}
   > n1 = Sequence()
   > UpdateBackground(StringMetadata=Get StringMetadata!() -> (StringMetadata))
   UpdateAlignment!(execute=@n1.then_1, self, StringMetadata:Map of Names to Strings=Get StringMetadata()) -> (then) {is=title K2Node_CallFunction}
  ]}
]
variables [
 EmptyBrush {type=Texture2D, default=&Empty1x1}
 ForegroundCenterPadding {type=Margin, default=<ltrb 0 0 0 0>}
 ForegroundLeftPadding {type=Margin, default=<ltrb 32 0 0 0>}
 ForegroundRightPadding {type=Margin, default=<ltrb 0 0 32 0>}
 IsActive {type=boolean}
 OnActivationComplete {}
 OnActivationStarted {}
 OnDeactivationComplete {}
 PreviewBackgroundImage {type=Texture2D, default=&/T_UI_Placeholder_Billboard_Background}
 PreviewForegroundImage {
  type=Texture2D
  default=/Game/UI/Foundation/Textures/BattleRoyale/FeaturedItems/Outfit/T-AthenaSoldiers-CID-722-Athena-M-TeriyakiFishAssassin-Bling
 }
 TextureManager {type=FortManagedTextureObject, default=null}
]
widgets {
 info {inheritedWidgetCount=0, parentClass=&UserWidget, rootWidgetClass=&Overlay, widgetCount=7}
 widgets [
  Overlay_0 {widgetClassPath=&Overlay, widget=Overlay_0}
  LazyImage_Background {
   widgetClassPath=&FortLazyImage
   bIsVariable=true
   parent=Overlay_0
   widget=LazyImage_Background
   properties {
    materialTextureParamName=Texture
    brush {
     bIsDynamicallyLoaded=false
     drawAs=Image
     imageType=FullColor
     mirroring=NoMirror
     resourceName=None
     resourceObject=&MI_UI_Billboard_Background
     tiling=NoTile
     imageSize=<32 32>
     margin=<ltrb 0 0 0 0>
     outlineSettings {
      bUseBrushTransparency=false
      roundingType=HalfHeightRadius
      width=0
      color {colorUseRule=UseColor_Specified, specifiedColor=<rgba 0 0 0 0>}
      cornerRadii=<0 0 0 0>
     }
     tintColor {colorUseRule=UseColor_Specified, specifiedColor=<rgba 1 1 1 1>}
     uVRegion {bIsValid=false, max=<0 0>, min=<0 0>}
    }
   }
   slot {class=OverlaySlot, properties {horizontalAlignment=HAlign_Fill, verticalAlignment=VAlign_Fill}}
  }
  WBP_CaptureForPostBufferUpdate {
   widgetClassPath=&WBP_CaptureForPostBufferUpdate
   parent=Overlay_0
   slot {class=OverlaySlot, properties {horizontalAlignment=HAlign_Fill, verticalAlignment=VAlign_Fill}}
  }
  Image_PostBufferBlur {
   widgetClassPath=&Image
   parent=Overlay_0
   widget=Image_PostBufferBlur
   properties {
    visibility=SelfHitTestInvisible
    brush {
     bIsDynamicallyLoaded=false
     drawAs=Image
     imageType=FullColor
     mirroring=NoMirror
     resourceName=None
     resourceObject=&/MI_UI_Billboard_Background_Blur
     tiling=NoTile
     imageSize=<32 32>
     margin=<ltrb 0 0 0 0>
     outlineSettings {
      bUseBrushTransparency=false
      roundingType=HalfHeightRadius
      width=0
      color {colorUseRule=UseColor_Specified, specifiedColor=<rgba 0 0 0 0>}
      cornerRadii=<0 0 0 0>
     }
     tintColor {colorUseRule=UseColor_Specified, specifiedColor=<rgba 1 1 1 1>}
     uVRegion {bIsValid=false, max=<0 0>, min=<0 0>}
    }
   }
   slot {class=OverlaySlot, properties {horizontalAlignment=HAlign_Fill, verticalAlignment=VAlign_Fill}}
  }
  Overlay_1 {
   widgetClassPath=&Overlay
   parent=Overlay_0
   widget=Overlay_1
   slot {class=OverlaySlot, properties {horizontalAlignment=HAlign_Center, verticalAlignment=VAlign_Bottom}}
  }
  Spacer {
   widgetClassPath=&Spacer
   bIsVariable=true
   parent=Overlay_1
   widget=Spacer
   properties {size=<1920 1080>}
   slot {class=OverlaySlot, properties {horizontalAlignment=HAlign_Center, verticalAlignment=VAlign_Bottom}}
  }
  LazyImage_Foreground {
   widgetClassPath=&FortLazyImage
   bIsVariable=true
   parent=Overlay_1
   widget=LazyImage_Foreground
   properties {
    materialTextureParamName=Texture
    brush {
     bIsDynamicallyLoaded=false
     drawAs=Image
     imageType=FullColor
     mirroring=NoMirror
     resourceName=None
     resourceObject=&/MI_UI_Billboard_Foreground
     tiling=NoTile
     imageSize=<1080 1080>
     margin=<ltrb 0 0 0 0>
     outlineSettings {
      bUseBrushTransparency=false
      roundingType=HalfHeightRadius
      width=0
      color {colorUseRule=UseColor_Specified, specifiedColor=<rgba 0 0 0 0>}
      cornerRadii=<0 0 0 0>
     }
     tintColor {colorUseRule=UseColor_Specified, specifiedColor=<rgba 1 1 1 1>}
     uVRegion {bIsValid=false, max=<0 0>, min=<0 0>}
    }
   }
   slot {class=OverlaySlot, properties {horizontalAlignment=HAlign_Fill, verticalAlignment=VAlign_Fill}}
  }
 ]
}
