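# WBP_Billboard_Banner is a User Widget that displays a banner image for an item shop multi-product billboard.
# A lazy-loading image sits in a height-fitting ScaleBox, overlaid with a Spacer that sets a nominal height.
# One function fills it from a soft texture reference, another from a CMS image URL.
# PreConstruct, only at design time, applies the PreviewBannerImage placeholder through the texture function.
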WidgetBlueprint /ItemShop/MultiProduct/Billboard/WBP_Billboard_Banner
graphs [
 EventGraph {nodes [
   n1 = Event PreConstruct()
   > Branch(Condition=@n1.IsDesignTime)
   > SetBannerImageFromTexture!(
    execute
    self
    LazyObject:"Soft Object Reference (by ref)"=Get PreviewBannerImage() -> (PreviewBannerImage:Soft Object Reference)
    bMatchSize:bool=false
   ) -> (then)
  ]}
 SetBannerImageFromTexture {nodes [
   n1 = SetBannerImageFromTexture!() -> (then, LazyObject:"Soft Object Reference (by ref)", bMatchSize:bool=~) {is=title K2Node_FunctionEntry}
   > SetBrushFromLazyDisplayAsset!(
    execute
    self:Common Lazy Image Object Reference=Get LazyImage_BannerImage() -> (LazyImage_BannerImage:Fort Lazy Image Object Reference)
    LazyObject:"Soft Object Reference (by ref)"=@n1
    bMatchTextureSize:bool=@n1.bMatchSize
   ) -> (then)
  ]}
 SetBannerImageFromURL {nodes [
   n1 = SetBannerImageFromURL!() -> (then, CMSImageURL:String, bMatchSize:bool=~) {is=title K2Node_FunctionEntry}
   > SetBrushFromURL!(
    execute
    self:Fort Lazy Image Object Reference=Get LazyImage_BannerImage()
    CMSImageURL:String=@n1
    bMatchSize:bool=@n1.bMatchSize
    Delay:float=0
    DownloadPriority:ECMSDownloadPriority Enum=Normal
   ) -> (then)
  ]}
]
variables [PreviewBannerImage {type=Object, default=&/T_UI_Billboard_BannerImagePlaceholder}]
widgets {
 info {inheritedWidgetCount=0, parentClass=&UserWidget, rootWidgetClass=&Overlay, widgetCount=4}
 widgets [
  Overlay_0 {widgetClassPath=&Overlay, widget=Overlay_0}
  ScaleBox_0 {widgetClassPath=&ScaleBox, parent=Overlay_0, widget=ScaleBox_0, properties {stretch=ScaleToFitY}, slot {class=OverlaySlot}}
  LazyImage_BannerImage {
   widgetClassPath=&FortLazyImage
   bIsVariable=true
   parent=ScaleBox_0
   widget=LazyImage_BannerImage
   properties {brush {
     bIsDynamicallyLoaded=false
     drawAs=Image
     imageType=FullColor
     mirroring=NoMirror
     resourceName=None
     resourceObject=&T_UI_Billboard_BannerImagePlaceholder
     tiling=NoTile
     imageSize=<264 48>
     margin=<ltrb 0 0 0 0>
     outlineSettings {
      bUseBrushTransparency=false
      roundingType=HalfHeightRadius
      width=0
      color {colorUseRule=UseColor_Specified, specifiedColor=<rgba 0 0 0 0>}
      cornerRadii=<0 0 0 0>
     }
     tintColor {colorUseRule=UseColor_Specified, specifiedColor=<rgba 1 1 1 1>}
     uVRegion {bIsValid=false, max=<0 0>, min=<0 0>}
    }}
   slot {class=ScaleBoxSlot}
  }
  Spacer_49 {widgetClassPath=&Spacer, parent=Overlay_0, widget=Spacer_49, properties {size=<1 48>}, slot {class=OverlaySlot}}
 ]
}
