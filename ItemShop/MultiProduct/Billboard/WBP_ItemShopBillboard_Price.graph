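# Widget for Item Shop billboards that shows either a styled price or an "Owned" label.
# A visibility switcher holds the "Owned" text and a vertical box with WBP_Price_ItemShop above a discount disclaimer, which is linked to that price widget.
# PreConstruct passes the optional real-price legal text block, when valid, into the price widget.
# GetSwitcherStateWidget returns the price widget or owned text based on IsOwned, without switching itself.
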WidgetBlueprint /ItemShop/MultiProduct/Billboard/WBP_ItemShopBillboard_Price
graphs [
 EventGraph {nodes [
   Event PreConstruct()
   > n1 = Get!(execute:Exec) -> (then:Exec, else:Exec, RealPriceLegalText_Optional:Common Text Object Reference)
   > Set RealPriceLegalText_Optional!(
    execute
    RealPriceLegalText_Optional:Common Text Object Reference=@n1
    self:WBP UIKit Price Object Reference=Get WBP_Price_ItemShop() -> (WBP_Price_ItemShop:WBP Price Item Shop Object Reference)
   ) -> (then, Output_Get:Common Text Object Reference)
  ]}
 GetSwitcherStateWidget {nodes [
   GetSwitcherStateWidget!() -> (then, IsOwned:bool=~) {is=title K2Node_FunctionEntry}
   > Return Node!(
    execute
    ReturnValue:Widget Object Reference=Select(
     Option 0:Widget Object Reference=Get WBP_Price_ItemShop()
     Option 1:Widget Object Reference=Get Text_Owned() -> (Text_Owned:Common Text Object Reference)
     Index=Get IsOwned() -> (IsOwned:bool)
    ) -> (ReturnValue:Widget Object Reference)
   )
  ]}
]
variables [RealPriceLegalText_Optional {type=CommonTextBlock, default=null}]
widgets {
 info {inheritedWidgetCount=0, parentClass=&UserWidget, rootWidgetClass=&CommonVisibilitySwitcher, widgetCount=5}
 widgets [
  Switcher_Ownership {widgetClassPath=&CommonVisibilitySwitcher, widget=Switcher_Ownership}
  VerticalBox_1 {
   widgetClassPath=&VerticalBox
   parent=Switcher_Ownership
   widget=VerticalBox_1
   slot {class=CommonVisibilitySwitcherSlot, properties {horizontalAlignment=HAlign_Center}}
  }
  WBP_Price_ItemShop {
   widgetClassPath=/ItemShop/MultiProduct/WBP_Price_ItemShop
   bIsVariable=true
   parent=VerticalBox_1
   widget=WBP_Price_ItemShop
   properties {
    discountDisclaimerText_Optional=&self:WidgetTree.DiscountDisclaimer
    additionalPriceFont {
     bForceMonospaced=false
     bMaterialIsStencil=false
     fontMaterial=/FortUILibrary/Systems/ItemShopTile/Materials/PriceSwitcher/MI_UIKit_PriceSwitcher_Text
     fontObject=&HeadingNow
     letterSpacing=0
     monospacedWidth=1
     size=15
     skewAmount=0
     typefaceFontName=75 Medium Italic
     outlineSettings {
      bApplyOutlineToDropShadows=false
      bMiteredCorners=false
      bSeparateFillAlpha=false
      outlineMaterial=null
      outlineSize=0
      outlineColor=<rgba 0 0 0 1>
     }
    }
    additionalPricePadding=<ltrb 4 0 4 3>
    additionalRegularPriceFont {
     bForceMonospaced=false
     bMaterialIsStencil=false
     fontMaterial=null
     fontObject=&HeadingNow
     letterSpacing=0
     monospacedWidth=1
     size=25.5
     skewAmount=0
     typefaceFontName=75 Medium
     outlineSettings {
      bApplyOutlineToDropShadows=false
      bMiteredCorners=false
      bSeparateFillAlpha=false
      outlineMaterial=null
      outlineSize=0
      outlineColor=<rgba 0 0 0 1>
     }
    }
    iconSize=<36 36>
    priceFont {
     bForceMonospaced=false
     bMaterialIsStencil=false
     fontMaterial=&MI_UIKit_PriceSwitcher_Text
     fontObject=&HeadingNow
     letterSpacing=0
     monospacedWidth=1
     size=25.5
     skewAmount=0
     typefaceFontName=86 Bold Italic
     outlineSettings {
      bApplyOutlineToDropShadows=false
      bMiteredCorners=false
      bSeparateFillAlpha=false
      outlineMaterial=null
      outlineSize=0
      outlineColor=<rgba 0 0 0 1>
     }
    }
    regularPriceFont {
     bForceMonospaced=false
     bMaterialIsStencil=false
     fontMaterial=&MI_UIKit_PriceSwitcher_Text
     fontObject=&HeadingNow
     letterSpacing=0
     monospacedWidth=1
     size=25.5
     skewAmount=0
     typefaceFontName=86 Bold Italic
     outlineSettings {
      bApplyOutlineToDropShadows=false
      bMiteredCorners=false
      bSeparateFillAlpha=false
      outlineMaterial=null
      outlineSize=0
      outlineColor=<rgba 0 0 0 1>
     }
    }
    regularPricePadding=<ltrb 4 0 4 0>
   }
   slot {class=VerticalBoxSlot}
  }
  DiscountDisclaimer {
   widgetClassPath=&CommonTextBlock
   bIsVariable=true
   parent=VerticalBox_1
   widget=DiscountDisclaimer
   properties {
    applyLineHeightToBottomLine=false
    bAutoCollapseWithEmptyText=true
    justification=Center
    mobileFontSizeMultiplier=1.2
    text="[PH]*Discount for select new players; full price applies to all others"
    font {
     bForceMonospaced=false
     bMaterialIsStencil=false
     fontMaterial=null
     fontObject=&HeadingNow
     letterSpacing=0
     monospacedWidth=1
     size=11.25
     skewAmount=0
     typefaceFontName=74 Regular
     outlineSettings {
      bApplyOutlineToDropShadows=false
      bMiteredCorners=false
      bSeparateFillAlpha=false
      outlineMaterial=null
      outlineSize=0
      outlineColor=<rgba 0 0 0 1>
     }
    }
    shadowOffset=<0 0>
   }
   slot {class=VerticalBoxSlot}
  }
  Text_Owned {
   widgetClassPath=&CommonTextBlock
   bIsVariable=true
   parent=Switcher_Ownership
   widget=Text_Owned
   properties {
    bAutoCollapseWithEmptyText=true
    justification=Center
    mobileFontSizeMultiplier=1.2
    text=Owned
    textTransformPolicy=ToUpper
    colorAndOpacity {colorUseRule=UseColor_Specified, specifiedColor=<rgba 0.0369 1 0.2232 1>}
    font {
     bForceMonospaced=false
     bMaterialIsStencil=false
     fontMaterial=&MI_UIKit_PriceSwitcher_Text
     fontObject=&HeadingNow
     letterSpacing=0
     monospacedWidth=1
     size=25.5
     skewAmount=0
     typefaceFontName=86 Bold Italic
     outlineSettings {
      bApplyOutlineToDropShadows=false
      bMiteredCorners=false
      bSeparateFillAlpha=false
      outlineMaterial=null
      outlineSize=0
      outlineColor=<rgba 0 0 0 1>
     }
    }
    shadowOffset=<0 0>
   }
   slot {class=CommonVisibilitySwitcherSlot}
  }
 ]
}
