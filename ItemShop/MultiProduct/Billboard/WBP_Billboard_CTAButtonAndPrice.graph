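# Item Shop billboard widget that pairs a call-to-action button with an offer price, separated by a spacer in a grid.
# PreConstruct applies button text, alignment, orientation and spacing, placing items side by side or stacking price above button.
# For Arabic culture it mirrors horizontal column order and left/right alignment.
# It relays clicks through OnCTAButtonClicked, passes the offer view model to the price widget, and can collapse the button with its spacer.
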WidgetBlueprint /ItemShop/MultiProduct/Billboard/WBP_Billboard_CTAButtonAndPrice
graphs [
 EventGraph {nodes [
   "On Clicked (WBP_UIKit_Button_Loud)"!() -> (OutputDelegate, then, Button:Common Button Base Object Reference) {is=title K2Node_ComponentBoundEvent}
   > Call On CTAButton Clicked!(execute:Exec, self:WBP Billboard CTAButton And Price Object Reference) -> (then:Exec) {is=title K2Node_CallDelegate}
   Event PreConstruct()
   > InitializeLayout!(
    execute
    self
    LayoutOrientation:E_Orientation Enum=Get LayoutOrientation() -> (LayoutOrientation:E_Orientation Enum=NewEnumerator0)
    LayoutSpacing:Vector 2D Structure=Get LayoutSpacing() -> (LayoutSpacing:Vector 2D Structure)
    HorizontalAlignment:EHorizontalAlignment Enum=Get HorizontalAlignment() -> (HorizontalAlignment:EHorizontalAlignment Enum=HAlign_Fill)
    VerticalAlignment:EVerticalAlignment Enum=Get VerticalAlignment() -> (VerticalAlignment:EVerticalAlignment Enum=VAlign_Fill)
   ) -> (then)
   > SetButtonText!(execute, self, ButtonText:Text=Get ButtonText() -> (ButtonText:Text)) -> (then)
  ]}
 InitializeLayout {nodes [
   n1 = InitializeLayout!() -> (
    then
    LayoutOrientation:E_Orientation Enum=NewEnumerator0
    LayoutSpacing:Vector 2D Structure
    HorizontalAlignment:EHorizontalAlignment Enum=HAlign_Fill
    VerticalAlignment:EVerticalAlignment Enum=VAlign_Fill
   ) {is=title K2Node_FunctionEntry}
   > Set LayoutOrientation(LayoutOrientation:E_Orientation Enum=@n1) -> (Output_Get:E_Orientation Enum=NewEnumerator0)
   > Set LayoutSpacing(LayoutSpacing:Vector 2D Structure=@n1.LayoutSpacing) -> (Output_Get:Vector 2D Structure)
   > Set HorizontalAlignment(HorizontalAlignment:EHorizontalAlignment Enum=@n1.HorizontalAlignment) -> (Output_Get:EHorizontalAlignment Enum=HAlign_Fill)
   > Set VerticalAlignment(VerticalAlignment:EVerticalAlignment Enum=@n1.VerticalAlignment) -> (Output_Get:EVerticalAlignment Enum=VAlign_Fill)
   > SetAlignment!(
    execute
    self
    HorizontalAlignment:EHorizontalAlignment Enum=Get HorizontalAlignment() -> (HorizontalAlignment=HAlign_Fill)
    VerticalAlignment:EVerticalAlignment Enum=Get VerticalAlignment() -> (VerticalAlignment=VAlign_Fill)
   ) -> (then)
   > SetLayoutOrientation!(execute, self, Orientation:E_Orientation Enum=Get LayoutOrientation() -> (LayoutOrientation=NewEnumerator0)) -> (then)
  ]}
 OnCTAButtonClicked {nodes [OnCTAButtonClicked() {is=title K2Node_FunctionEntry}]}
 SetAlignment {nodes [
   n1 = SetAlignment!() -> (then, HorizontalAlignment:EHorizontalAlignment Enum=HAlign_Fill, VerticalAlignment:EVerticalAlignment Enum=VAlign_Fill) {is=title K2Node_FunctionEntry}
   > Set HorizontalAlignment(HorizontalAlignment=@n1) -> (Output_Get=HAlign_Fill)
   > Set VerticalAlignment(VerticalAlignment=@n1.VerticalAlignment) -> (Output_Get=VAlign_Fill)
   > n2 = Branch(Condition="Equal Exactly (String)"(A=GetCurrentCulture!() -> (ReturnValue:String), B=ar))
   n3 = Get HorizontalAlignment() -> (HorizontalAlignment=HAlign_Fill)
   > n4 = Set L_LocalizedHorizontalAlignment(L_LocalizedHorizontalAlignment:EHorizontalAlignment Enum=Select!(
     HAlign_Fill:EHorizontalAlignment Enum=HAlign_Fill
     HAlign_Left:EHorizontalAlignment Enum=HAlign_Right
     HAlign_Center:EHorizontalAlignment Enum=HAlign_Center
     HAlign_Right:EHorizontalAlignment Enum=HAlign_Left
     Index:EHorizontalAlignment Enum=@n3
    ) -> (ReturnValue:EHorizontalAlignment Enum)) -> (Output_Get:EHorizontalAlignment Enum=HAlign_Fill)
   SetGridSlotAlignment!(
    execute=[@n4.then, @n5.then]
    Widget:Widget Object Reference=Get WBP_ItemShopBillboard_Price() -> (WBP_ItemShopBillboard_Price:WBP Item Shop Billboard Price Object Reference)
    Horizontal:EHorizontalAlignment Enum=Get L_LocalizedHorizontalAlignment() -> (L_LocalizedHorizontalAlignment:EHorizontalAlignment Enum=HAlign_Fill)
    Vertical:EVerticalAlignment Enum=Get VerticalAlignment() -> (VerticalAlignment=VAlign_Fill)
   ) -> (then)
   > SetGridSlotAlignment(
    Widget=Get WBP_UIKit_Button_Loud() -> (WBP_UIKit_Button_Loud:WBP UIKit Button Loud Object Reference)
    Horizontal=Get L_LocalizedHorizontalAlignment() -> (L_LocalizedHorizontalAlignment=HAlign_Fill)
    Vertical=Get VerticalAlignment() -> (VerticalAlignment=VAlign_Fill)
   )
   n5 = Set L_LocalizedHorizontalAlignment(execute=@n2.else, L_LocalizedHorizontalAlignment=@n3) -> (Output_Get=HAlign_Fill)
  ]}
 SetButtonText {nodes [
   n1 = SetButtonText!() -> (then, ButtonText:Text) {is=title K2Node_FunctionEntry}
   > Set ButtonText(ButtonText:Text=@n1) -> (Output_Get:Text)
   > SetText!(execute, self:Fort CTAButton Object Reference=Get WBP_UIKit_Button_Loud(), ButtonText:"Text (by ref)"=Get ButtonText()) -> (then)
  ]}
 SetCTAButtonVisibility {nodes [
   SetCTAButtonVisibility!() -> (then, ShouldShow:bool=~) {is=title K2Node_FunctionEntry}
   > SetVisibility(
    self=[Get WBP_UIKit_Button_Loud(), Get Spacing() -> (Spacing:Spacer Object Reference)]
    InVisibility=Select!(
     Option 0:ESlateVisibility Enum=Collapsed
     Option 1:ESlateVisibility Enum=SelfHitTestInvisible
     Index:bool=Get ShouldShow() -> (ShouldShow:bool)
    ) -> (ReturnValue:ESlateVisibility Enum)
   )
  ]}
 SetLayoutOrientation {nodes [
   n1 = SetLayoutOrientation!() -> (then, Orientation:E_Orientation Enum=NewEnumerator0) {is=title K2Node_FunctionEntry}
   > n2 = Set LayoutOrientation(LayoutOrientation=@n1) -> (Output_Get=NewEnumerator0)
   > n3 = Switch on E_Orientation!(execute, Selection:E_Orientation Enum=@n2) -> (NewEnumerator0:Exec, NewEnumerator1:Exec) {is=title K2Node_SwitchEnum}
   n4 = "Equal Exactly (String)"(A=GetCurrentCulture(), B=ar)
   > SetGridSlotColumnAndRow!(
    execute
    Widget:Widget Object Reference=Get WBP_UIKit_Button_Loud()
    Row:int=0
    Row Span:int=0
    Column:int=SelectInt(A=2, B=0, bPickA=@n4)
    Column Span:int=0
   ) -> (then)
   > SetGridSlotColumnAndRow(Widget=Get Spacing(), Row=0, Row Span=0, Column=1, Column Span=0)
   > n5 = SetGridSlotColumnAndRow(Widget=Get WBP_ItemShopBillboard_Price(), Row=0, Row Span=0, Column=SelectInt(A=0, B=2, bPickA=@n4), Column Span=0)
   SetSpacing!(execute=[@n5.then, @n6.then], self, LayoutSpacing:Vector 2D Structure=Get LayoutSpacing()) -> (then)
   SetGridSlotColumnAndRow(execute=@n3.NewEnumerator1, Widget=Get WBP_UIKit_Button_Loud(), Row=2, Row Span=0, Column=0, Column Span=0)
   > SetGridSlotColumnAndRow(Widget=Get Spacing(), Row=1, Row Span=0, Column=0, Column Span=0)
   > n6 = SetGridSlotColumnAndRow(Widget=Get WBP_ItemShopBillboard_Price(), Row=0, Row Span=0, Column=0, Column Span=0)
  ]}
 SetPriceSwitcherVM {nodes [
   n1 = SetPriceSwitcherVM!() -> (then, OfferVM:Fort Item Shop Offer VM Object Reference) {is=title K2Node_FunctionEntry}
   > Set FortItemShopOfferVM!(
    execute
    FortItemShopOfferVM:Fort Item Shop Offer VM Object Reference=@n1
    self:WBP Item Shop Billboard Price Object Reference=Get WBP_ItemShopBillboard_Price()
   ) -> (then, Output_Get:Fort Item Shop Offer VM Object Reference)
  ]}
 SetSpacing {nodes [
   n1 = SetSpacing!() -> (then, LayoutSpacing:Vector 2D Structure) {is=title K2Node_FunctionEntry}
   > Set LayoutSpacing(LayoutSpacing=@n1)
   n2 = Get LayoutOrientation() -> (LayoutOrientation=NewEnumerator0)
   n3 = Get!() -> (LayoutSpacing_X:double, LayoutSpacing_Y:double)
   > SetSize!(
    execute
    self:Spacer Object Reference=Get Spacing()
    InSize_X:double=Select!(NewEnumerator0:double=@n3, NewEnumerator1:double=1, Index:E_Orientation Enum=@n2) -> (ReturnValue:double)
    InSize_Y:double=Select(NewEnumerator0=1, NewEnumerator1=@n3.LayoutSpacing_Y, Index=@n2)
   ) -> (then)
  ]}
]
variables [
 ButtonText {type=string, default="[PH] View Offer"}
 HorizontalAlignment {type=EHorizontalAlignment, default=HAlign_Center}
 LayoutOrientation {type=E_Orientation, default=Horizontal}
 LayoutSpacing {type=Vector2D, default=<40 84>}
 OnCTAButtonClicked {}
 VerticalAlignment {type=EVerticalAlignment, default=VAlign_Center}
]
widgets {
 info {inheritedWidgetCount=0, parentClass=&UserWidget, rootWidgetClass=&GridPanel, widgetCount=4}
 widgets [
  GridPanel_Root {widgetClassPath=&GridPanel, bIsVariable=true, widget=GridPanel_Root}
  WBP_UIKit_Button_Loud {
   widgetClassPath=/FortUILibrary/UIKit/Buttons/WBP_UIKit_Button_Loud
   bIsVariable=true
   parent=GridPanel_Root
   widget=WBP_UIKit_Button_Loud
   properties {minHeight=56}
   slot {class=GridSlot}
  }
  Spacing {widgetClassPath=&Spacer, bIsVariable=true, parent=GridPanel_Root, widget=Spacing, slot {class=GridSlot, properties {column=1}}}
  WBP_ItemShopBillboard_Price {
   widgetClassPath=&/WBP_ItemShopBillboard_Price
   bIsVariable=true
   parent=GridPanel_Root
   widget=WBP_ItemShopBillboard_Price
   slot {class=GridSlot, properties {column=2, row=1, verticalAlignment=VAlign_Center}}
  }
 ]
}
