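# Item Shop billboard details widget showing an offer's title, subtitle, highlight, banner and body images, item cards and CTA button.
# View-model and metadata values feed text, title gradient colors and lazily created banner widgets in named slots.
# Layout mirrors alignment for Arabic, adjusts padding on mobile and shrinks oversized title fonts.
# Staggered ease-in animations follow scroll direction, reversing for ease-out; CTA clicks log analytics, broadcast and set the featured offer.
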WidgetBlueprint /ItemShop/MultiProduct/Billboard/WBP_Billboard_Details
animations [
 EaseIn {
  binding_count=7
  length_seconds=1.1
  object=&self:EaseIn
  sequence {bindings [
    Text_Subtitle {
     id=#1
     possessed_class=CommonTextBlock
     tracks [
      {
       type=MovieSceneFloatTrack
       display_name=Render Opacity
       sections [{type=MovieSceneFloatSection, end_frame=20, start_frame=0, channels [None {type=MovieSceneScriptingFloatChannel, keys [[0,0],[15,0],[20,1]]}]}]
      }
      {
       type=MovieScene2DTransformTrack
       sections [{
         type=MovieScene2DTransformSection
         end_frame=0
         start_frame=0
         channels [
          Translation.X {type=MovieSceneScriptingFloatChannel, const [0,1]}
          Translation.Y {type=MovieSceneScriptingFloatChannel, const [0,1]}
          Angle {type=MovieSceneScriptingFloatChannel, const [0,1]}
          Scale.X {type=MovieSceneScriptingFloatChannel, const [1,1]}
          Scale.Y {type=MovieSceneScriptingFloatChannel, const [1,1]}
          Shear.X {type=MovieSceneScriptingFloatChannel, const [0,1]}
          Shear.Y {type=MovieSceneScriptingFloatChannel, const [0,1]}
         ]
        }]
      }
     ]
    }
    Switcher_ItemCards {
     id=#2
     possessed_class=CommonVisibilitySwitcher
     tracks [
      {
       type=MovieSceneFloatTrack
       display_name=Render Opacity
       sections [{type=MovieSceneFloatSection, end_frame=21, start_frame=0, channels [None {type=MovieSceneScriptingFloatChannel, keys [[0,0],[16,0],[21,1]]}]}]
      }
      {
       type=MovieScene2DTransformTrack
       sections [{
         type=MovieScene2DTransformSection
         end_frame=0
         start_frame=0
         channels [
          Translation.X {type=MovieSceneScriptingFloatChannel, const [0,1]}
          Translation.Y {type=MovieSceneScriptingFloatChannel, const [0,1]}
         ]
        }]
      }
     ]
    }
    NamedSlot_BodyImage {
     id=#3
     possessed_class=NamedSlot
     tracks [
      {
       type=MovieSceneFloatTrack
       display_name=Render Opacity
       sections [{type=MovieSceneFloatSection, end_frame=21, start_frame=0, channels [None {type=MovieSceneScriptingFloatChannel, keys [[0,0],[16,0],[21,1]]}]}]
      }
      {
       type=MovieScene2DTransformTrack
       sections [{
         type=MovieScene2DTransformSection
         end_frame=0
         start_frame=0
         channels [
          Translation.X {type=MovieSceneScriptingFloatChannel, const [0,1]}
          Translation.Y {type=MovieSceneScriptingFloatChannel, const [0,1]}
         ]
        }]
      }
     ]
    }
    WBP_Billboard_CTAButtonAndPrice {
     id=#4
     possessed_class=WBP_Billboard_CTAButtonAndPrice_C
     tracks [
      {
       type=MovieSceneFloatTrack
       display_name=Render Opacity
       sections [{
         type=MovieSceneFloatSection
         end_frame=22
         start_frame=0
         channels [None {type=MovieSceneScriptingFloatChannel, keys [[0,0],[17.2177,0],[22,1]]}]
        }]
      }
      {
       type=MovieScene2DTransformTrack
       sections [{
         type=MovieScene2DTransformSection
         end_frame=0
         start_frame=0
         channels [
          Translation.X {type=MovieSceneScriptingFloatChannel, const [0,1]}
          Translation.Y {type=MovieSceneScriptingFloatChannel, const [0,1]}
          Angle {type=MovieSceneScriptingFloatChannel, const [0,1]}
          Scale.X {type=MovieSceneScriptingFloatChannel, const [1,1]}
          Scale.Y {type=MovieSceneScriptingFloatChannel, const [1,1]}
          Shear.X {type=MovieSceneScriptingFloatChannel, const [0,1]}
          Shear.Y {type=MovieSceneScriptingFloatChannel, const [0,1]}
         ]
        }]
      }
     ]
    }
    NamedSlot_BannerImage {
     id=#5
     possessed_class=NamedSlot
     tracks [
      {
       type=MovieSceneFloatTrack
       display_name=Render Opacity
       sections [{type=MovieSceneFloatSection, end_frame=17, start_frame=0, channels [None {type=MovieSceneScriptingFloatChannel, keys [[0,0],[12,0],[17,1]]}]}]
      }
      {
       type=MovieScene2DTransformTrack
       sections [{
         type=MovieScene2DTransformSection
         end_frame=0
         start_frame=0
         channels [
          Translation.X {type=MovieSceneScriptingFloatChannel, const [0,1]}
          Translation.Y {type=MovieSceneScriptingFloatChannel, const [0,1]}
          Angle {type=MovieSceneScriptingFloatChannel, const [0,1]}
          Scale.X {type=MovieSceneScriptingFloatChannel, const [1,1]}
          Scale.Y {type=MovieSceneScriptingFloatChannel, const [1,1]}
          Shear.X {type=MovieSceneScriptingFloatChannel, const [0,1]}
          Shear.Y {type=MovieSceneScriptingFloatChannel, const [0,1]}
         ]
        }]
      }
     ]
    }
    TitleSwitcher {
     id=#6
     display_name=CommonVisibilitySwitcher_0
     possessed_class=CommonVisibilitySwitcher
     tracks [
      {
       type=MovieSceneFloatTrack
       display_name=Render Opacity
       sections [{type=MovieSceneFloatSection, end_frame=19, start_frame=0, channels [None {type=MovieSceneScriptingFloatChannel, keys [[0,0],[14,0],[19,1]]}]}]
      }
      {
       type=MovieScene2DTransformTrack
       sections [{
         type=MovieScene2DTransformSection
         end_frame=0
         start_frame=0
         channels [
          Translation.X {type=MovieSceneScriptingFloatChannel, const [0,1]}
          Translation.Y {type=MovieSceneScriptingFloatChannel, const [0,1]}
          Angle {type=MovieSceneScriptingFloatChannel, const [0,1]}
          Scale.X {type=MovieSceneScriptingFloatChannel, const [1,1]}
          Scale.Y {type=MovieSceneScriptingFloatChannel, const [1,1]}
          Shear.X {type=MovieSceneScriptingFloatChannel, const [0,1]}
          Shear.Y {type=MovieSceneScriptingFloatChannel, const [0,1]}
         ]
        }]
      }
     ]
    }
    NamedSlot_Highlight {
     id=#7
     possessed_class=NamedSlot
     tracks [
      {
       type=MovieSceneFloatTrack
       display_name=Render Opacity
       sections [{type=MovieSceneFloatSection, end_frame=18, start_frame=0, channels [None {type=MovieSceneScriptingFloatChannel, keys [[0,0],[13,0],[18,1]]}]}]
      }
      {
       type=MovieScene2DTransformTrack
       sections [{
         type=MovieScene2DTransformSection
         end_frame=0
         start_frame=0
         channels [
          Translation.X {type=MovieSceneScriptingFloatChannel, const [0,1]}
          Translation.Y {type=MovieSceneScriptingFloatChannel, const [0,1]}
          Angle {type=MovieSceneScriptingFloatChannel, const [0,1]}
          Scale.X {type=MovieSceneScriptingFloatChannel, const [1,1]}
          Scale.Y {type=MovieSceneScriptingFloatChannel, const [1,1]}
          Shear.X {type=MovieSceneScriptingFloatChannel, const [0,1]}
          Shear.Y {type=MovieSceneScriptingFloatChannel, const [0,1]}
         ]
        }]
      }
     ]
    }
   ]}
 }
 EaseIn_ScrolledUp {
  binding_count=8
  length_seconds=1.1
  object=&self:EaseIn_ScrolledUp
  sequence {bindings [
    Switcher_ItemCards {
     id=#8
     possessed_class=CommonVisibilitySwitcher
     tracks [
      {
       type=MovieSceneFloatTrack
       display_name=Render Opacity
       sections [{type=MovieSceneFloatSection, end_frame=21, start_frame=0, channels [None {type=MovieSceneScriptingFloatChannel, keys [[0,0],[16,0],[21,1]]}]}]
      }
      {
       type=MovieScene2DTransformTrack
       sections [{
         type=MovieScene2DTransformSection
         end_frame=21
         start_frame=0
         channels [Translation.Y {type=MovieSceneScriptingFloatChannel, keys [[0,-24],[16,-24],[21,0]]}]
        }]
      }
     ]
    }
    TitleSwitcher {
     id=#9
     display_name=CommonVisibilitySwitcher_0
     possessed_class=CommonVisibilitySwitcher
     tracks [
      {
       type=MovieSceneFloatTrack
       display_name=Render Opacity
       sections [{type=MovieSceneFloatSection, end_frame=19, start_frame=0, channels [None {type=MovieSceneScriptingFloatChannel, keys [[0,0],[14,0],[19,1]]}]}]
      }
      {
       type=MovieScene2DTransformTrack
       sections [{
         type=MovieScene2DTransformSection
         end_frame=19
         start_frame=0
         channels [
          Translation.X {type=MovieSceneScriptingFloatChannel, const [0,1]}
          Translation.Y {type=MovieSceneScriptingFloatChannel, keys [[0,-24],[14,-24],[19,0]]}
          Angle {type=MovieSceneScriptingFloatChannel, const [0,1]}
          Scale.X {type=MovieSceneScriptingFloatChannel, const [1,1]}
          Scale.Y {type=MovieSceneScriptingFloatChannel, const [1,1]}
          Shear.X {type=MovieSceneScriptingFloatChannel, const [0,1]}
          Shear.Y {type=MovieSceneScriptingFloatChannel, const [0,1]}
         ]
        }]
      }
     ]
    }
    NamedSlot_BodyImage {
     id=#10
     possessed_class=NamedSlot
     tracks [
      {
       type=MovieSceneFloatTrack
       display_name=Render Opacity
       sections [{type=MovieSceneFloatSection, end_frame=21, start_frame=0, channels [None {type=MovieSceneScriptingFloatChannel, keys [[0,0],[16,0],[21,1]]}]}]
      }
      {
       type=MovieScene2DTransformTrack
       sections [{
         type=MovieScene2DTransformSection
         end_frame=21
         start_frame=0
         channels [Translation.Y {type=MovieSceneScriptingFloatChannel, keys [[0,-24],[16,-24],[21,0]]}]
        }]
      }
     ]
    }
    Text_Subtitle {
     id=#1
     possessed_class=CommonTextBlock
     tracks [
      {
       type=MovieSceneFloatTrack
       display_name=Render Opacity
       sections [{type=MovieSceneFloatSection, end_frame=20, start_frame=0, channels [None {type=MovieSceneScriptingFloatChannel, keys [[0,0],[15,0],[20,1]]}]}]
      }
      {
       type=MovieScene2DTransformTrack
       sections [{
         type=MovieScene2DTransformSection
         end_frame=20
         start_frame=0
         channels [
          Translation.X {type=MovieSceneScriptingFloatChannel, const [0,1]}
          Translation.Y {type=MovieSceneScriptingFloatChannel, keys [[0,-24],[15,-24],[20,0]]}
          Angle {type=MovieSceneScriptingFloatChannel, const [0,1]}
          Scale.X {type=MovieSceneScriptingFloatChannel, const [1,1]}
          Scale.Y {type=MovieSceneScriptingFloatChannel, const [1,1]}
          Shear.X {type=MovieSceneScriptingFloatChannel, const [0,1]}
          Shear.Y {type=MovieSceneScriptingFloatChannel, const [0,1]}
         ]
        }]
      }
     ]
    }
    WBP_Billboard_CTAButtonAndPrice {
     id=#4
     possessed_class=WBP_Billboard_CTAButtonAndPrice_C
     tracks [
      {
       type=MovieSceneFloatTrack
       display_name=Render Opacity
       sections [{
         type=MovieSceneFloatSection
         end_frame=22
         start_frame=0
         channels [None {type=MovieSceneScriptingFloatChannel, keys [[0,0],[17,0],[21.9997,1]]}]
        }]
      }
      {
       type=MovieScene2DTransformTrack
       sections [{
         type=MovieScene2DTransformSection
         end_frame=21
         start_frame=0
         channels [
          Translation.X {type=MovieSceneScriptingFloatChannel, const [0,1]}
          Translation.Y {type=MovieSceneScriptingFloatChannel, keys [[0,-24],[17,-24],[21.9997,0]]}
          Angle {type=MovieSceneScriptingFloatChannel, const [0,1]}
          Scale.X {type=MovieSceneScriptingFloatChannel, const [1,1]}
          Scale.Y {type=MovieSceneScriptingFloatChannel, const [1,1]}
          Shear.X {type=MovieSceneScriptingFloatChannel, const [0,1]}
          Shear.Y {type=MovieSceneScriptingFloatChannel, const [0,1]}
         ]
        }]
      }
     ]
    }
    NamedSlot_BannerImage {
     id=#5
     possessed_class=NamedSlot
     tracks [
      {
       type=MovieScene2DTransformTrack
       sections [{
         type=MovieScene2DTransformSection
         end_frame=17
         start_frame=0
         channels [
          Translation.X {type=MovieSceneScriptingFloatChannel, const [0,1]}
          Translation.Y {type=MovieSceneScriptingFloatChannel, keys [[0,-24],[12,-24],[17,0]]}
          Angle {type=MovieSceneScriptingFloatChannel, const [0,1]}
          Scale.X {type=MovieSceneScriptingFloatChannel, const [1,1]}
          Scale.Y {type=MovieSceneScriptingFloatChannel, const [1,1]}
          Shear.X {type=MovieSceneScriptingFloatChannel, const [0,1]}
          Shear.Y {type=MovieSceneScriptingFloatChannel, const [0,1]}
         ]
        }]
      }
      {
       type=MovieSceneFloatTrack
       display_name=Render Opacity
       sections [{type=MovieSceneFloatSection, end_frame=17, start_frame=0, channels [None {type=MovieSceneScriptingFloatChannel, keys [[0,0],[12,0],[17,1]]}]}]
      }
     ]
    }
    NamedSlot_Highlight {
     id=#7
     possessed_class=NamedSlot
     tracks [
      {
       type=MovieSceneFloatTrack
       display_name=Render Opacity
       sections [{type=MovieSceneFloatSection, end_frame=18, start_frame=0, channels [None {type=MovieSceneScriptingFloatChannel, keys [[0,0],[13,0],[18,1]]}]}]
      }
      {
       type=MovieScene2DTransformTrack
       sections [{
         type=MovieScene2DTransformSection
         end_frame=18
         start_frame=0
         channels [
          Translation.X {type=MovieSceneScriptingFloatChannel, const [0,1]}
          Translation.Y {type=MovieSceneScriptingFloatChannel, keys [[0,-24],[13,-24],[18,0]]}
          Angle {type=MovieSceneScriptingFloatChannel, const [0,1]}
          Scale.X {type=MovieSceneScriptingFloatChannel, const [1,1]}
          Scale.Y {type=MovieSceneScriptingFloatChannel, const [1,1]}
          Shear.X {type=MovieSceneScriptingFloatChannel, const [0,1]}
          Shear.Y {type=MovieSceneScriptingFloatChannel, const [0,1]}
         ]
        }]
      }
     ]
    }
   ]}
 }
 EaseIn_ScrolledDown {
  binding_count=7
  length_seconds=1.1
  object=&self:EaseIn_ScrolledDown
  sequence {bindings [
    Text_Subtitle {
     id=#1
     possessed_class=CommonTextBlock
     tracks [
      {
       type=MovieSceneFloatTrack
       display_name=Render Opacity
       sections [{
         type=MovieSceneFloatSection
         end_frame=19
         start_frame=0
         channels [None {type=MovieSceneScriptingFloatChannel, thinned_from=4, keys [[0,0],[14,0],[19,1]]}]
        }]
      }
      {
       type=MovieScene2DTransformTrack
       sections [{
         type=MovieScene2DTransformSection
         end_frame=19
         start_frame=0
         channels [
          Translation.X {type=MovieSceneScriptingFloatChannel, const [0,1]}
          Translation.Y {type=MovieSceneScriptingFloatChannel, keys [[0,24],[14,24],[19,0]]}
          Angle {type=MovieSceneScriptingFloatChannel, const [0,1]}
          Scale.X {type=MovieSceneScriptingFloatChannel, const [1,1]}
          Scale.Y {type=MovieSceneScriptingFloatChannel, const [1,1]}
          Shear.X {type=MovieSceneScriptingFloatChannel, const [0,1]}
          Shear.Y {type=MovieSceneScriptingFloatChannel, const [0,1]}
         ]
        }]
      }
     ]
    }
    NamedSlot_BodyImage {
     id=#11
     possessed_class=NamedSlot
     tracks [
      {
       type=MovieSceneFloatTrack
       display_name=Render Opacity
       sections [{type=MovieSceneFloatSection, end_frame=18, start_frame=0, channels [None {type=MovieSceneScriptingFloatChannel, keys [[0,0],[13,0],[18,1]]}]}]
      }
      {
       type=MovieScene2DTransformTrack
       sections [{
         type=MovieScene2DTransformSection
         end_frame=18
         start_frame=0
         channels [Translation.Y {type=MovieSceneScriptingFloatChannel, keys [[0,24],[13,24],[18,0]]}]
        }]
      }
     ]
    }
    WBP_Billboard_CTAButtonAndPrice {
     id=#4
     possessed_class=WBP_Billboard_CTAButtonAndPrice_C
     tracks [
      {
       type=MovieSceneFloatTrack
       display_name=Render Opacity
       sections [{type=MovieSceneFloatSection, end_frame=17, start_frame=0, channels [None {type=MovieSceneScriptingFloatChannel, keys [[0,0],[12,0],[17,1]]}]}]
      }
      {
       type=MovieScene2DTransformTrack
       sections [{
         type=MovieScene2DTransformSection
         end_frame=17
         start_frame=0
         channels [
          Translation.X {type=MovieSceneScriptingFloatChannel, const [0,1]}
          Translation.Y {type=MovieSceneScriptingFloatChannel, keys [[0,24],[12,24],[17,0]]}
          Angle {type=MovieSceneScriptingFloatChannel, const [0,1]}
          Scale.X {type=MovieSceneScriptingFloatChannel, const [1,1]}
          Scale.Y {type=MovieSceneScriptingFloatChannel, const [1,1]}
          Shear.X {type=MovieSceneScriptingFloatChannel, const [0,1]}
          Shear.Y {type=MovieSceneScriptingFloatChannel, const [0,1]}
         ]
        }]
      }
     ]
    }
    NamedSlot_BannerImage {
     id=#5
     possessed_class=NamedSlot
     tracks [
      {
       type=MovieScene2DTransformTrack
       sections [{
         type=MovieScene2DTransformSection
         end_frame=22
         start_frame=0
         channels [
          Translation.X {type=MovieSceneScriptingFloatChannel, const [0,1]}
          Translation.Y {type=MovieSceneScriptingFloatChannel, keys [[0,24],[17,24],[21.9997,0]]}
          Angle {type=MovieSceneScriptingFloatChannel, const [0,1]}
          Scale.X {type=MovieSceneScriptingFloatChannel, const [1,1]}
          Scale.Y {type=MovieSceneScriptingFloatChannel, const [1,1]}
          Shear.X {type=MovieSceneScriptingFloatChannel, const [0,1]}
          Shear.Y {type=MovieSceneScriptingFloatChannel, const [0,1]}
         ]
        }]
      }
      {
       type=MovieSceneFloatTrack
       display_name=Render Opacity
       sections [{
         type=MovieSceneFloatSection
         end_frame=22
         start_frame=0
         channels [None {type=MovieSceneScriptingFloatChannel, keys [[0,0],[17,0],[21.9997,1]]}]
        }]
      }
     ]
    }
    Switcher_ItemCards {
     id=#12
     possessed_class=CommonVisibilitySwitcher
     tracks [
      {
       type=MovieSceneFloatTrack
       display_name=Render Opacity
       sections [{type=MovieSceneFloatSection, end_frame=18, start_frame=0, channels [None {type=MovieSceneScriptingFloatChannel, keys [[0,0],[13,0],[18,1]]}]}]
      }
      {
       type=MovieScene2DTransformTrack
       sections [{
         type=MovieScene2DTransformSection
         end_frame=18
         start_frame=0
         channels [Translation.Y {type=MovieSceneScriptingFloatChannel, keys [[0,24],[13,24],[18,0]]}]
        }]
      }
     ]
    }
    TitleSwitcher {
     id=#13
     display_name=CommonVisibilitySwitcher_0
     possessed_class=CommonVisibilitySwitcher
     tracks [
      {
       type=MovieSceneFloatTrack
       display_name=Render Opacity
       sections [{type=MovieSceneFloatSection, end_frame=20, start_frame=0, channels [None {type=MovieSceneScriptingFloatChannel, keys [[0,0],[15,0],[20,1]]}]}]
      }
      {
       type=MovieScene2DTransformTrack
       sections [{
         type=MovieScene2DTransformSection
         end_frame=20
         start_frame=0
         channels [
          Translation.X {type=MovieSceneScriptingFloatChannel, const [0,1]}
          Translation.Y {type=MovieSceneScriptingFloatChannel, keys [[0,24],[15,24],[20,0]]}
          Angle {type=MovieSceneScriptingFloatChannel, const [0,1]}
          Scale.X {type=MovieSceneScriptingFloatChannel, const [1,1]}
          Scale.Y {type=MovieSceneScriptingFloatChannel, const [1,1]}
          Shear.X {type=MovieSceneScriptingFloatChannel, const [0,1]}
          Shear.Y {type=MovieSceneScriptingFloatChannel, const [0,1]}
         ]
        }]
      }
     ]
    }
    NamedSlot_Highlight {
     id=#7
     possessed_class=NamedSlot
     tracks [
      {
       type=MovieSceneFloatTrack
       display_name=Render Opacity
       sections [{type=MovieSceneFloatSection, end_frame=21, start_frame=0, channels [None {type=MovieSceneScriptingFloatChannel, keys [[0,0],[16,0],[21,1]]}]}]
      }
      {
       type=MovieScene2DTransformTrack
       sections [{
         type=MovieScene2DTransformSection
         end_frame=21
         start_frame=0
         channels [
          Translation.X {type=MovieSceneScriptingFloatChannel, const [0,1]}
          Translation.Y {type=MovieSceneScriptingFloatChannel, keys [[0,24],[16,24],[21,0]]}
          Angle {type=MovieSceneScriptingFloatChannel, const [0,1]}
          Scale.X {type=MovieSceneScriptingFloatChannel, const [1,1]}
          Scale.Y {type=MovieSceneScriptingFloatChannel, const [1,1]}
          Shear.X {type=MovieSceneScriptingFloatChannel, const [0,1]}
          Shear.Y {type=MovieSceneScriptingFloatChannel, const [0,1]}
         ]
        }]
      }
     ]
    }
   ]}
 }
]
graphs [
 DesignTimePreview {nodes [
   DesignTimePreview() {is=title K2Node_FunctionEntry}
   > VM_HighlightTextUpdated!(execute, self, HighlightText:Text="LOCTEXT(\"[PH] HIghlight\")") -> (then)
   > UpdateBannerFromImage!(
    execute
    self
    BannerInstance:"WBP Billboard Banner Object Reference (by ref)"=Get BannerImageWidget() -> (BannerImageWidget:WBP Billboard Banner Object Reference)
    BannerSlot:Panel Widget Object Reference=Get NamedSlot_BannerImage() -> (NamedSlot_BannerImage:Named Slot Object Reference)
    PreviewImage:"Soft Object Reference (by ref)"=Get PreviewLazyBannerImage() -> (PreviewLazyBannerImage:Soft Object Reference)
    bMatchSize:bool=true
   ) -> (then)
   > UpdateBannerFromImage(
    BannerInstance=Get TitleImageWidget() -> (TitleImageWidget:WBP Billboard Banner Object Reference)
    BannerSlot=Get NamedSlot_TitleImage() -> (NamedSlot_TitleImage:Named Slot Object Reference)
    PreviewImage=Get PreviewLazyTitleImage() -> (PreviewLazyTitleImage:Soft Object Reference)
    bMatchSize=true
   )
   > UpdateBannerFromImage(
    BannerInstance=Get BodyImageWidget() -> (BodyImageWidget:WBP Billboard Banner Object Reference)
    BannerSlot=Get NamedSlot_BodyImage() -> (NamedSlot_BodyImage:Named Slot Object Reference)
    PreviewImage=Get PreviewLazyBodyImage() -> (PreviewLazyBodyImage:Soft Object Reference)
    bMatchSize=true
   )
   > SetWidthOverride!(execute, self:Self Object Reference, WidthOverride:float=Get WidthOverride() -> (WidthOverride:float)) -> (then)
   > UpdateTitleFontSize()
  ]}
 Ease in Text {nodes [
   n1 = Ease in Text!() -> (then, IsImmediate:bool=~) {is=title K2Node_FunctionEntry}
   > StopAllAnimations(self:User Widget Object Reference)
   > n2 = Branch(Condition=@n1)
   n3 = Select!(
    NoScroll:Widget Animation Object Reference=Get EaseIn() -> (EaseIn:Widget Animation Object Reference)
    ScrolledDown:Widget Animation Object Reference=Get EaseIn_ScrolledDown() -> (EaseIn_ScrolledDown:Widget Animation Object Reference)
    ScrolledUp:Widget Animation Object Reference=Get EaseIn_ScrolledUp() -> (EaseIn_ScrolledUp:Widget Animation Object Reference)
    Index:EFortItemShopScrollDirection Enum=Get LastScrolledDirection!(self:Fort Item Shop Selection VM Object Reference=Get FortItemShopSelectionVM() -> (FortItemShopSelectionVM:Fort Item Shop Selection VM Object Reference)) -> (LastScrolledDirection:EFortItemShopScrollDirection Enum=NoScroll)
   ) -> (ReturnValue:Widget Animation Object Reference)
   > PlayAnimation(InAnimation=@n3, StartAtTime=GetEndTime(self=@n3))
   PlayAnimation(execute=@n2.else, InAnimation=@n3)
  ]}
 Ease out Text {nodes [
   n1 = Ease out Text!() -> (then, IsImmediate:bool=~) {is=title K2Node_FunctionEntry}
   > StopAllAnimations()
   > n2 = Branch(Condition=@n1)
   n3 = Select(
    NoScroll=Get EaseIn()
    ScrolledDown=Get EaseIn_ScrolledDown()
    ScrolledUp=Get EaseIn_ScrolledUp()
    Index=Get LastScrolledDirection(self=Get FortItemShopSelectionVM()) -> (LastScrolledDirection=NoScroll)
   )
   > PlayAnimation(InAnimation=@n3, StartAtTime=GetEndTime(self=@n3), PlayMode=Reverse)
   PlayAnimation(execute=@n2.else, InAnimation=@n3, PlayMode=Reverse)
  ]}
 EventGraph {nodes [
   "On CTAButton Clicked (WBP_Billboard_CTAButtonAndPrice)"() {is=title K2Node_ComponentBoundEvent}
   > n1 = Get!(execute:Exec) -> (then:Exec, else:Exec, FortItemShopOfferVM:Fort Item Shop Offer VM Object Reference)
   > Fire Event Element Interacted(
    Element=Get WBP_Billboard_CTAButtonAndPrice() -> (WBP_Billboard_CTAButtonAndPrice:WBP Billboard CTAButton And Price Object Reference)
    AdditionalInfo="Add String (Non-Localized)"!(
     Info:"Flow Analytics Additional Info Structure (by ref)"=Make Flow Analytics Additional Info!() -> (FlowAnalyticsAdditionalInfo:Flow Analytics Additional Info Structure) {is=title K2Node_MakeStruct}
     Key:String=Display Asset Path
     Value:String=Get DisplayAssetPath!(self:Fort Item Shop Offer VM Object Reference=@n1) -> (DisplayAssetPath:String)
    ) -> (ReturnValue:Flow Analytics Additional Info Structure)
   )
   > Call On CTAButton Clicked!(execute:Exec, self:WBP Billboard Details Object Reference, OfferVM:Fort Item Shop Offer VM Object Reference=@n1) -> (then:Exec) {is=title K2Node_CallDelegate}
   > Set FeaturedOfferId!(
    execute
    FeaturedOfferId:String=Get OfferId!(self:Fort Item Shop Offer VM Object Reference=@n1) -> (OfferId:String)
    self:Fort Item Shop VM Object Reference=Get FortItemShopVM() -> (FortItemShopVM:Fort Item Shop VM Object Reference)
   ) -> (then, Output_Get:String)
   n2 = Event PreConstruct()
   > n3 = Branch(Condition=@n2.IsDesignTime)
   > n4 = DesignTimePreview() {is=title K2Node_CallFunction}
   SetLayoutAlignment!(
    execute=[@n5.then, @n4.then]
    self
    Horizontal Alignment:EHorizontalAlignment Enum=Get HorizontalAlignment() -> (HorizontalAlignment:EHorizontalAlignment Enum=HAlign_Fill)
   ) -> (then)
   > SetMobileLayout()
   SetWidthOverride(execute=@n3.else, WidthOverride=Get WidthOverride())
   > n5 = Ease out Text!(execute, self, IsImmediate:bool=true) -> (then) {is=title K2Node_CallFunction}
  ]}
 GetBodyImageURL {nodes [
   n1 = GetBodyImageURL!() -> (then, InStringMetadata:Map of Names to Strings, InShopContext:Gameplay Tag Structure) {is=title K2Node_FunctionEntry}
   > n2 = FindItemShopConfig!(execute, ProductTag:Gameplay Tag Structure=@n1.InShopContext) -> (then, ReturnValue:Fort Product Item Shop Config Object Reference)
   > n3 = Is Valid(InputObject=@n2)
   n4 = Find(
    TargetMap:Map of Names to Strings=@n1
    Key:"Name (by ref)"=Get BodyImageMetadataName!(self:Fort Product Item Shop Config Object Reference=@n2) -> (BodyImageMetadataName:Name=None)
   )
   > n5 = Branch(Condition=@n4.ReturnValue)
   > Return Node!(execute, URL:String=@n4)
   n6 = Find(TargetMap=@n1, Key=BodyImage)
   Return Node(execute=[@n5.else, @n3.Is Not Valid], URL=Select!(Option 0:String, Option 1:String=@n6, Index:bool=@n6.ReturnValue) -> (ReturnValue:String))
  ]}
 GetDesiredFocusTarget {nodes [
   GetDesiredFocusTarget() {is=title K2Node_FunctionEntry}
   > Return Node!(execute, Target:Widget Object Reference=Get WBP_Billboard_CTAButtonAndPrice())
  ]}
 GetSmartCardsVisibility {nodes [
   n1 = GetSmartCardsVisibility!() -> (
    then
    InSecondaryGrantedItems:Array of Fort Item VM Object References
    InStringMetadata:Map of Names to Strings
    InShopContext:Gameplay Tag Structure
   ) {is=title K2Node_FunctionEntry}
   n2 = Is Empty!(TargetArray:Array of Fort Item VM Object References=@n1) -> (ReturnValue:bool)
   n3 = IsEmpty(InString=GetBodyImageURL!(self, InStringMetadata:Map of Names to Strings=@n1.InStringMetadata, InShopContext:Gameplay Tag Structure=@n1.InShopContext) -> (URL:String) {is=title K2Node_CallFunction})
   > Return Node!(
    execute
    Visibility:ESlateVisibility Enum=Select(Option 0:ESlateVisibility Enum=SelfHitTestInvisible, Option 1:ESlateVisibility Enum=Collapsed, Index=AND Boolean(A=@n2, B=@n3)) -> (ReturnValue:ESlateVisibility Enum)
   )
  ]}
 OnCTAButtonClicked {nodes [OnCTAButtonClicked!() -> (then, OfferVM:Fort Item Shop Offer VM Object Reference) {is=title K2Node_FunctionEntry}]}
 SetLayoutAlignment {nodes [
   n1 = SetLayoutAlignment!() -> (then, Horizontal Alignment:EHorizontalAlignment Enum=HAlign_Fill) {is=title K2Node_FunctionEntry}
   > Set HorizontalAlignment(HorizontalAlignment:EHorizontalAlignment Enum=@n1) -> (Output_Get:EHorizontalAlignment Enum=HAlign_Fill)
   > n2 = Branch(Condition="Equal Exactly (String)"(A=GetCurrentCulture!() -> (ReturnValue:String), B=ar))
   n3 = Get HorizontalAlignment() -> (HorizontalAlignment=HAlign_Fill)
   > n4 = Set L_LocalizedHorizontalAlignment(L_LocalizedHorizontalAlignment:EHorizontalAlignment Enum=Select!(
     HAlign_Fill:EHorizontalAlignment Enum=HAlign_Fill
     HAlign_Left:EHorizontalAlignment Enum=HAlign_Right
     HAlign_Center:EHorizontalAlignment Enum=HAlign_Center
     HAlign_Right:EHorizontalAlignment Enum=HAlign_Left
     Index:EHorizontalAlignment Enum=@n3
    ) -> (ReturnValue:EHorizontalAlignment Enum)) -> (Output_Get:EHorizontalAlignment Enum=HAlign_Fill)
   n5 = Sequence!(execute=[@n11.then, @n4.then]) -> (then_0, then_1, then_2:Exec)
   > n6 = Set HorizontalAlignment!(
    execute
    HorizontalAlignment=Get L_LocalizedHorizontalAlignment() -> (L_LocalizedHorizontalAlignment:EHorizontalAlignment Enum=HAlign_Fill)
    self:Vertical Box Slot Object Reference=SlotAsVerticalBoxSlot(Widget=Get NamedSlot_BannerImage())
   ) -> (then, Output_Get=HAlign_Fill)
   > n7 = Set HorizontalAlignment(HorizontalAlignment=@n6, self=SlotAsVerticalBoxSlot(Widget=Get NamedSlot_Highlight() -> (NamedSlot_Highlight:Named Slot Object Reference))) -> (Output_Get=HAlign_Fill)
   > n8 = Set HorizontalAlignment(HorizontalAlignment=@n7, self=SlotAsVerticalBoxSlot(Widget=Get WBP_Billboard_CTAButtonAndPrice())) -> (Output_Get=HAlign_Fill)
   > n9 = Set HorizontalAlignment(
    HorizontalAlignment=@n8
    self=SlotAsVerticalBoxSlot(Widget=Get Switcher_ItemCards() -> (Switcher_ItemCards:Common Visibility Switcher Object Reference))
   ) -> (Output_Get=HAlign_Fill)
   > Set HorizontalAlignment(HorizontalAlignment=@n9, self=SlotAsVerticalBoxSlot(Widget=Get NamedSlot_TitleImage())) -> (Output_Get=HAlign_Fill)
   SetLayoutOrientation!(
    execute=@n5.then_2
    self:WBP Billboard CTAButton And Price Object Reference=Get WBP_Billboard_CTAButtonAndPrice()
    Orientation:E_Orientation Enum=Select(
     HAlign_Fill:E_Orientation Enum=NewEnumerator0
     HAlign_Left:E_Orientation Enum=NewEnumerator0
     HAlign_Center:E_Orientation Enum=NewEnumerator1
     HAlign_Right:E_Orientation Enum=NewEnumerator0
     Index=Get L_LocalizedHorizontalAlignment() -> (L_LocalizedHorizontalAlignment=HAlign_Fill)
    ) -> (ReturnValue:E_Orientation Enum)
   ) -> (then)
   n10 = Set Justification!(
    execute=@n5.then_1
    Justification:ETextJustify Enum=Select(
     HAlign_Fill:ETextJustify Enum=Left
     HAlign_Left:ETextJustify Enum=Left
     HAlign_Center:ETextJustify Enum=Center
     HAlign_Right:ETextJustify Enum=Right
     Index=Get HorizontalAlignment() -> (HorizontalAlignment=HAlign_Fill)
    ) -> (ReturnValue:ETextJustify Enum)
    self:Text Layout Widget Object Reference=Get Text_Title() -> (Text_Title:Common Text Object Reference)
   ) -> (then, Output_Get:ETextJustify Enum=Left)
   > Set Justification(Justification=@n10, self=Get Text_Subtitle() -> (Text_Subtitle:Common Text Object Reference)) -> (Output_Get=Left)
   n11 = Set L_LocalizedHorizontalAlignment(execute=@n2.else, L_LocalizedHorizontalAlignment=@n3) -> (Output_Get=HAlign_Fill)
  ]}
 SetMobileLayout {nodes [
   SetMobileLayout() {is=title K2Node_FunctionEntry}
   > Branch(Condition=IsMobileGame())
   n1 = Get WBP_Billboard_CTAButtonAndPrice()
   > Set LayoutSpacing!(execute, LayoutSpacing:Vector 2D Structure=(X=40,Y=24), self:WBP Billboard CTAButton And Price Object Reference=@n1) -> (then, Output_Get:Vector 2D Structure)
   n2 = Get!() -> (Padding_Left:float=~, Padding_Top:float=~, Padding_Right:float=~, Padding_Bottom:float=~)
   > SetPadding!(
    execute
    self:Vertical Box Slot Object Reference=SlotAsVerticalBoxSlot(Widget=@n1)
    InPadding_Left:float=@n2
    InPadding_Top:float=24
    InPadding_Right:float=@n2.Padding_Right
    InPadding_Bottom:float=@n2.Padding_Bottom
   ) -> (then)
   n3 = Get() -> (Padding_Left=~, Padding_Top=~, Padding_Right=~, Padding_Bottom=~)
   > SetPadding(
    self=SlotAsVerticalBoxSlot(Widget=Get NamedSlot_BannerImage())
    InPadding_Left=@n3
    InPadding_Top=@n3.Padding_Top
    InPadding_Right=@n3.Padding_Right
    InPadding_Bottom=24
   )
   n4 = Get() -> (Padding_Left=~, Padding_Top=~, Padding_Right=~, Padding_Bottom=~)
   > SetPadding(
    self=SlotAsVerticalBoxSlot(Widget=Get Switcher_ItemCards())
    InPadding_Left=@n4
    InPadding_Top=@n4.Padding_Top
    InPadding_Right=@n4.Padding_Right
    InPadding_Bottom=24
   )
   n5 = Get() -> (Padding_Left=~, Padding_Top=~, Padding_Right=~, Padding_Bottom=~)
   > SetPadding(
    self=SlotAsVerticalBoxSlot(Widget=Get Text_Subtitle())
    InPadding_Left=@n5
    InPadding_Top=@n5.Padding_Top
    InPadding_Right=@n5.Padding_Right
    InPadding_Bottom=24
   )
  ]}
 SetTextFromMetadata {nodes [
   SetTextFromMetadata!() -> (then, Key:"Name (by ref)", TextWidget:Text Object Reference) {is=title K2Node_FunctionEntry}
   n1 = Find(TargetMap:Map of Names to Texts=Get Text Metadata() -> (Text Metadata:Map of Names to Texts), Key=Get Key() -> (Key:"Name (by ref)"=None)) -> (Value:Text)
   n2 = AND Boolean(A=@n1.ReturnValue, B=NOT Boolean(A=TextIsEmpty(InText=@n1)))
   > n3 = Branch(Condition=@n2)
   n4 = Get TextWidget() -> (TextWidget:Text Object Reference)
   > "SetText (Text)"(self=@n4, InText=@n1)
   > SetVisibility(self=@n4, InVisibility=HitTestInvisible)
   > Return Node!(execute, isTextSet:bool=@n2)
   SetVisibility(execute=@n3.else, self=@n4, InVisibility=Collapsed)
  ]}
 SetTitleColor {nodes [
   n1 = SetTitleColor!() -> (then, StringMetadata:Map of Names to Strings) {is=title K2Node_FunctionEntry}
   > n2 = GetTitleColorsFromStringMetadata!(execute, StringMetadata:Map of Names to Strings=@n1) -> (then, ColorA:Linear Color Structure, ColorB:Linear Color Structure, Found:bool)
   > Branch(Condition=@n2.Found)
   > Set ColorA(ColorA:Linear Color Structure=@n2) -> (Output_Get:Linear Color Structure)
   > Set ColorB(ColorB:Linear Color Structure=@n2.ColorB) -> (Output_Get:Linear Color Structure)
   > n3 = GetDynamicFontMaterial!(execute, self:Text Object Reference=Get Text_Title()) -> (then, ReturnValue:Material Instance Dynamic Object Reference)
   > SetVectorParameterValue(self=@n3, ParameterName=Gradient Color 1, Value=Get ColorA() -> (ColorA:Linear Color Structure))
   > SetVectorParameterValue(self=@n3, ParameterName=Gradient Color 2, Value=Get ColorB() -> (ColorB:Linear Color Structure))
  ]}
 SetWidthOverride {nodes [
   n1 = SetWidthOverride!() -> (then, WidthOverride:float=~) {is=title K2Node_FunctionEntry}
   > n2 = Set WidthOverride(WidthOverride:float=@n1) -> (Output_Get:float)
   > Set WidthOverride!(execute, WidthOverride=@n2, self:Size Box Object Reference=Get SizeBox_All() -> (SizeBox_All:Size Box Object Reference)) -> (then, Output_Get)
   > SetWrapTextWidth!(execute, self:Common Text Object Reference=Get Text_Title(), InWrapTextAt:int=Truncate(A=Get WidthOverride())) -> (then)
   > UpdateTitleFontSize()
   > n3 = Get!(execute:Exec) -> (then:Exec, else:Exec, HighlightWidget:WBP Item Shop Tile Highlight 2 Object Reference)
   > SetWidth!(execute, self:WBP UIKit Highlight Base Object Reference=@n3, NewWidthOverride:double=Get WidthOverride()) -> (then)
  ]}
 UpdateBannerFromImage {nodes [
   n1 = UpdateBannerFromImage!() -> (
    then
    BannerInstance:"WBP Billboard Banner Object Reference (by ref)"
    BannerSlot:Panel Widget Object Reference
    PreviewImage:"Soft Object Reference (by ref)"
    bMatchSize:bool=~
   ) {is=title K2Node_FunctionEntry}
   > n2 = Is Valid(InputObject=@n1)
   n3 = Create Widget!(
    execute:Exec=@n2.Is Not Valid
    Class:User Widget Class Reference=&/WBP_Billboard_Banner.WBP_Billboard_Banner_C
    OwningPlayer:Player Controller Object Reference=GetOwningPlayer()
   ) -> (then:Exec, ReturnValue:WBP Billboard Banner Object Reference)
   > "Set WBP Billboard Banner Object Reference (by ref)"!(
    execute:Exec
    Target:"WBP Billboard Banner Object Reference (by ref)"=Get BannerInstance() -> (BannerInstance:"WBP Billboard Banner Object Reference (by ref)")
    Value:WBP Billboard Banner Object Reference=@n3
   ) -> (then:Exec) {is=title K2Node_VariableSetRef}
   > AddChild(self=Get BannerSlot() -> (BannerSlot:Panel Widget Object Reference), Content=Get BannerInstance())
   > SetBannerImageFromTexture!(
    execute
    self:WBP Billboard Banner Object Reference=Get BannerInstance()
    LazyObject:"Soft Object Reference (by ref)"=Get PreviewImage() -> (PreviewImage:"Soft Object Reference (by ref)")
    bMatchSize:bool=Get bMatchSize() -> (bMatchSize:bool)
   ) -> (then)
   SetBannerImageFromTexture(execute=@n2.Is Valid, self=Get BannerInstance(), LazyObject=Get PreviewImage(), bMatchSize=Get bMatchSize())
  ]}
 UpdateBannerFromURL {nodes [
   n1 = UpdateBannerFromURL!() -> (
    then
    BannerInstance:"WBP Billboard Banner Object Reference (by ref)"
    BannerSlot:Panel Widget Object Reference
    URL:"String (by ref)"
    bMatchSize:bool=~
   ) {is=title K2Node_FunctionEntry}
   > n2 = Branch(Condition=IsEmpty(InString=@n1.URL))
   > n3 = Is Valid(InputObject=Get BannerInstance())
   > n4 = Set Visibility(Visibility=Collapsed, self=Get BannerInstance())
   Set Visibility(execute=[@n4.then, @n3.Is Not Valid], Visibility=Collapsed, self=Get BannerSlot())
   n5 = Is Valid(exec=@n2.else, InputObject=Get BannerInstance())
   n6 = Create Widget(execute=@n5.Is Not Valid, Class=&WBP_Billboard_Banner_C, OwningPlayer=GetOwningPlayer())
   > "Set WBP Billboard Banner Object Reference (by ref)"(Target=Get BannerInstance(), Value=@n6)
   > n7 = AddChild(self=Get BannerSlot(), Content=Get BannerInstance())
   SetBannerImageFromURL!(
    execute=[@n7.then, @n5.Is Valid]
    self:WBP Billboard Banner Object Reference=Get BannerInstance()
    CMSImageURL:String=Get URL() -> (URL:"String (by ref)")
    bMatchSize:bool=Get bMatchSize()
   ) -> (then)
   > Set Visibility(Visibility=SelfHitTestInvisible, self=Get BannerInstance())
   > Set Visibility(Visibility=SelfHitTestInvisible, self=Get BannerSlot())
  ]}
 UpdateTitleFontSize {nodes [
   UpdateTitleFontSize() {is=title K2Node_FunctionEntry}
   > n1 = GetDynamicFontMaterial(self=Get Text_Title())
   > Set TitleFontMID(TitleFontMID:Material Instance Dynamic Object Reference=@n1) -> (Output_Get:Material Instance Dynamic Object Reference)
   > SetFont(self=Get Text_Title(), InFontInfo=Get TitleFontDefault() -> (TitleFontDefault:Slate Font Info Structure))
   n2 = Get Text_Title()
   > ForceLayoutPrepass(self:Widget Object Reference=@n2)
   n3 = Get SizeBox_Title() -> (SizeBox_Title:Size Box Object Reference)
   > n4 = Branch(Condition="float > float"(
     A=GetDesiredSize!(self:Widget Object Reference=@n2) -> (ReturnValue_X:double, ReturnValue_Y:double).ReturnValue_Y
     B=Get MaxDesiredHeight!(self:Size Box Object Reference=@n3) -> (MaxDesiredHeight:float)
    ))
   > n5 = SetFont(self=Get Text_Title(), InFontInfo=Get TitleFontSmall() -> (TitleFontSmall:Slate Font Info Structure))
   SetFontMaterial!(
    execute=[@n5.then, @n4.else]
    self:Text Object Reference=Get Text_Title()
    InMaterial:Material Interface Object Reference=Get TitleFontMID() -> (TitleFontMID:Material Instance Dynamic Object Reference)
   ) -> (then)
  ]}
 VM_BodyImageUpdated {nodes [
   n1 = VM_BodyImageUpdated!() -> (then, CMS Image URL:String) {is=title K2Node_FunctionEntry}
   > UpdateBannerFromURL!(
    execute
    self
    BannerInstance:"WBP Billboard Banner Object Reference (by ref)"=Get BodyImageWidget()
    BannerSlot:Panel Widget Object Reference=Get NamedSlot_BodyImage()
    URL:"String (by ref)"=@n1
    bMatchSize:bool=true
   ) -> (then) {is=title K2Node_CallFunction}
   > SetActiveWidget(
    self=Get Switcher_ItemCards()
    Widget=Select!(
     Option 0:Widget Object Reference=Get NamedSlot_BodyImage()
     Option 1:Widget Object Reference=Get SmartCardWrapper() -> (SmartCardWrapper:WBP Item Shop Tile Smart Card Wrapper Object Reference)
     Index:bool=IsEmpty(InString=@n1)
    ) -> (ReturnValue:Widget Object Reference)
   )
  ]}
 VM_HighlightTextUpdated {nodes [
   n1 = VM_HighlightTextUpdated!() -> (then, HighlightText:Text) {is=title K2Node_FunctionEntry}
   n2 = Find(TargetMap:Map of Names to Strings=Get String Metadata() -> (String Metadata:Map of Names to Strings), Key=highlightEnabled) -> (Value:String)
   n3 = "Equal Exactly (String)"(A=@n2, B=false)
   > n4 = Branch(Condition=OR Boolean(A=TextIsEmpty(InText=@n1), B=AND Boolean(A=@n3, B=@n2.ReturnValue)))
   > Set Visibility(Visibility=Collapsed, self=Get NamedSlot_Highlight())
   n5 = Get(execute=@n4.else)
   > n6 = SetViewModelByClass!(
    execute
    Widget:User Widget Object Reference=Get HighlightWidget() -> (HighlightWidget:WBP Item Shop Tile Highlight 2 Object Reference)
    ViewModel:Notify Field Value Changed Interface=Get FortItemShopOfferVM() -> (FortItemShopOfferVM:Fort Item Shop Offer VM Object Reference)
   ) -> (then)
   Set Visibility(execute=[@n9.then, @n6.then], Visibility=SelfHitTestInvisible, self=Get NamedSlot_Highlight())
   n7 = Construct Object from Class!(
    execute=@n5.else
    Class=/ItemShop/MultiProduct/ItemShopTileComponents/WBP_ItemShopTile_Highlight_2.WBP_ItemShopTile_Highlight_2_C
    self
    FortItemShopOfferVM:Fort Item Shop Offer VM Object Reference=Get FortItemShopOfferVM()
   ) -> (then, ReturnValue:WBP Item Shop Tile Highlight 2 Object Reference)
   > n8 = Set HighlightWidget(HighlightWidget:WBP Item Shop Tile Highlight 2 Object Reference=@n7) -> (Output_Get:WBP Item Shop Tile Highlight 2 Object Reference)
   > SetWidth(self=@n8, NewWidthOverride=Get WidthOverride())
   > n9 = AddChild(self=Get NamedSlot_Highlight(), Content=Get HighlightWidget())
  ]}
 VM_OffersUpdated {nodes [
   n1 = VM_OffersUpdated!() -> (then, Offers:Array of Fort Item Shop Offer VM Object References) {is=title K2Node_FunctionEntry}
   > n2 = Branch(Condition=Is Valid Index!(TargetArray:Array of Fort Item Shop Offer VM Object References=@n1, IndexToTest:int) -> (ReturnValue:bool))
   > SetFortItemShopOfferVM!(
    execute
    self
    Viewmodel:Fort Item Shop Offer VM Object Reference="Get (a copy)"(Array:Array of Fort Item Shop Offer VM Object References=@n1) -> (Output:Fort Item Shop Offer VM Object Reference)
   ) -> (then)
   > SetPriceSwitcherVM!(
    execute
    self:WBP Billboard CTAButton And Price Object Reference=Get WBP_Billboard_CTAButtonAndPrice()
    OfferVM:Fort Item Shop Offer VM Object Reference=Get FortItemShopOfferVM()
   ) -> (then)
   SetFortItemShopOfferVM(execute=@n2.else)
   > SetPriceSwitcherVM(self=Get WBP_Billboard_CTAButtonAndPrice())
  ]}
 VM_StringMetadataUpdated {nodes [
   n1 = VM_StringMetadataUpdated!() -> (then, StringMetadata:Map of Names to Strings) {is=title K2Node_FunctionEntry}
   > Set String Metadata(String Metadata:Map of Names to Strings=@n1) -> (Output_Get:Map of Names to Strings)
   > n2 = Sequence!(execute) -> (then_0, then_1, then_2, then_3:Exec, then_4:Exec)
   SetTitleColor!(execute=@n2.then_1, self, StringMetadata:Map of Names to Strings=Get String Metadata()) -> (then) {is=title K2Node_CallFunction}
   n3 = GetAlignmentFromStringMetadata!(execute=@n2.then_0, StringMetadata:Map of Names to Strings=Get String Metadata()) -> (then, Alignment:EHorizontalAlignment Enum=HAlign_Fill, Found:bool)
   > Branch(Condition=@n3.Found)
   > SetLayoutAlignment(Horizontal Alignment=Select!(
     HAlign_Fill:EHorizontalAlignment Enum=HAlign_Fill
     HAlign_Left:EHorizontalAlignment Enum=HAlign_Left
     HAlign_Center:EHorizontalAlignment Enum=HAlign_Center
     HAlign_Right:EHorizontalAlignment Enum=HAlign_Left
     Index:EHorizontalAlignment Enum=@n3
    ) -> (ReturnValue:EHorizontalAlignment Enum)) {is=title K2Node_CallFunction}
   > SetWidthOverride(WidthOverride=Select(HAlign_Fill:float=766, HAlign_Left:float=766, HAlign_Center:float=960, HAlign_Right:float=894, Index=@n3) -> (ReturnValue:float)) {is=title K2Node_CallFunction}
   n4 = Get!(execute=@n2.then_4) -> (then, else, FortItemShopOfferVM:Fort Item Shop Offer VM Object Reference)
   n5 = TextIsEmpty(InText=GetHighlightText!(self:Fort Item Shop Offer VM Object Reference=Get FortItemShopOfferVM()) -> (ReturnValue:Text))
   n6 = Find(TargetMap=Get String Metadata(), Key=highlightEnabled)
   n7 = AND Boolean(A="Equal Exactly (String)"(A=@n6, B=false), B=@n6.ReturnValue)
   > SetVisibility(
    self=Get NamedSlot_Highlight()
    InVisibility=Select!(Option 0:ESlateVisibility Enum=SelfHitTestInvisible, Option 1:ESlateVisibility Enum=Collapsed, Index:bool=OR Boolean(A=@n7, B=@n5)) -> (ReturnValue:ESlateVisibility Enum)
   )
   SetVisibility(execute=@n4.else, self=Get NamedSlot_Highlight(), InVisibility=Collapsed)
   n8 = Find(TargetMap=Get String Metadata(), Key=titleImage)
   UpdateBannerFromURL(execute=@n2.then_3, BannerInstance=Get TitleImageWidget(), BannerSlot=Get NamedSlot_TitleImage(), URL=@n8, bMatchSize=true)
   > Branch(Condition=@n8.ReturnValue)
   > SetActiveWidget(self=Get TitleSwitcher() -> (TitleSwitcher:Common Visibility Switcher Object Reference), Widget=Get NamedSlot_TitleImage())
   UpdateBannerFromURL(
    execute=@n2.then_2
    BannerInstance=Get BannerImageWidget()
    BannerSlot=Get NamedSlot_BannerImage()
    URL=Find(TargetMap=Get String Metadata(), Key=banner)
    bMatchSize=true
   )
  ]}
 VM_TextMetadataUpdated {nodes [
   n1 = VM_TextMetadataUpdated!() -> (then, TextMetadata:Map of Names to Texts) {is=title K2Node_FunctionEntry}
   > Set Text Metadata(Text Metadata:Map of Names to Texts=@n1) -> (Output_Get:Map of Names to Texts)
   > n2 = Sequence!(execute) -> (then_0, then_1, then_2)
   > n3 = SetTextFromMetadata!(execute, self, Key:"Name (by ref)"=MakeLiteralName(Value=title), TextWidget:Text Object Reference=Get Text_Title()) -> (then, isTextSet:bool) {is=title K2Node_CallFunction}
   > Branch(Condition=@n3)
   > SetActiveWidget(self=Get TitleSwitcher(), Widget=Get SizeBox_Title())
   > UpdateTitleFontSize() {is=title K2Node_CallFunction}
   SetTextFromMetadata(execute=@n2.then_1, Key=MakeLiteralName(Value=subtitle), TextWidget=Get Text_Subtitle())
   n4 = Find(TargetMap:Map of Names to Texts=Get TextMetadata() -> (TextMetadata:Map of Names to Texts), Key=cta) -> (Value:Text)
   n5 = Branch(execute=@n2.then_2, Condition=@n4.ReturnValue)
   n6 = Get WBP_Billboard_CTAButtonAndPrice()
   > SetButtonText!(execute, self:WBP Billboard CTAButton And Price Object Reference=@n6, ButtonText:Text=@n4) -> (then)
   SetButtonText(execute=@n5.else, self=@n6, ButtonText="LOCTEXT(\"View Offer\")")
  ]}
]
variables [
 BannerImageWidget {type=WBP_Billboard_Banner_C, default=null}
 BodyImageWidget {type=WBP_Billboard_Banner_C, default=null}
 HighlightWidget {type=WBP_ItemShopTile_Highlight_2_C, default=null}
 HorizontalAlignment {type=EHorizontalAlignment, default=HAlign_Left}
 OnCTAButtonClicked {}
 PreviewLazyBannerImage {type=Object, default=&/T_UI_Billboard_BannerImagePlaceholder}
 PreviewLazyBodyImage {type=Object, default=&/T_UI_Billboard_BodyImagePlaceholder}
 PreviewLazyTitleImage {type=Object, default=&/T_UI_Billboard_TitleImagePlaceholder}
 String Metadata {type="map<string>", default {}}
 Text Metadata {type="map<string>", default {}}
 TitleFontDefault {
  type=SlateFontInfo
  default {
   bForceMonospaced=false
   bMaterialIsStencil=false
   fontMaterial=&/M_UI_Billboard_Title
   fontObject=&HeadingNow
   letterSpacing=0
   monospacedWidth=1
   size=72.75
   skewAmount=0
   typefaceFontName=86 Bold Italic
   outlineSettings {
    bApplyOutlineToDropShadows=false
    bMiteredCorners=false
    bSeparateFillAlpha=false
    outlineMaterial=null
    outlineSize=0
    outlineColor=<rgba 0 0 0 1>
   }
  }
 }
 TitleFontSmall {
  type=SlateFontInfo
  default {
   bForceMonospaced=false
   bMaterialIsStencil=false
   fontMaterial=&M_UI_Billboard_Title
   fontObject=&HeadingNow
   letterSpacing=0
   monospacedWidth=1
   size=58.5
   skewAmount=0
   typefaceFontName=86 Bold Italic
   outlineSettings {
    bApplyOutlineToDropShadows=false
    bMiteredCorners=false
    bSeparateFillAlpha=false
    outlineMaterial=null
    outlineSize=0
    outlineColor=<rgba 0 0 0 1>
   }
  }
 }
 TitleImageWidget {type=WBP_Billboard_Banner_C, default=null}
 WidthOverride {type=number, default=900}
]
widgets {
 info {inheritedWidgetCount=0, parentClass=&UserWidget, rootWidgetClass=&SizeBox, widgetCount=15}
 widgets [
  SizeBox_All {widgetClassPath=&SizeBox, bIsVariable=true, widget=SizeBox_All, properties {bOverride_WidthOverride=true, widthOverride=1196}}
  VerticalBox_All {
   widgetClassPath=&VerticalBox
   parent=SizeBox_All
   widget=VerticalBox_All
   slot {class=SizeBoxSlot, properties {verticalAlignment=VAlign_Bottom}}
  }
  VerticalBox_0 {
   widgetClassPath=&VerticalBox
   parent=VerticalBox_All
   widget=VerticalBox_0
   slot {class=VerticalBoxSlot, properties {horizontalAlignment=HAlign_Left, padding=<ltrb 0 48 0 0>}}
  }
  NamedSlot_Highlight {widgetClassPath=&NamedSlot, bIsVariable=true, parent=VerticalBox_0, widget=NamedSlot_Highlight, slot {class=VerticalBoxSlot}}
  NamedSlot_BannerImage {
   widgetClassPath=&NamedSlot
   bIsVariable=true
   parent=VerticalBox_0
   widget=NamedSlot_BannerImage
   properties {visibility=Collapsed}
   slot {class=VerticalBoxSlot, properties {padding=<ltrb 0 24 0 0>}}
  }
  VerticalBox_1 {
   widgetClassPath=&VerticalBox
   parent=VerticalBox_All
   widget=VerticalBox_1
   slot {class=VerticalBoxSlot, properties {padding=<ltrb 0 48 0 0>}}
  }
  TitleSwitcher {
   widgetClassPath=&CommonVisibilitySwitcher
   bIsVariable=true
   parent=VerticalBox_1
   widget=TitleSwitcher
   properties {bOverride_Cursor=true, shownVisibility=HitTestInvisible}
   slot {class=VerticalBoxSlot}
  }
  SizeBox_Title {
   widgetClassPath=&SizeBox
   bIsVariable=true
   parent=TitleSwitcher
   widget=SizeBox_Title
   properties {bOverride_MaxDesiredHeight=true, maxDesiredHeight=350}
   slot {class=CommonVisibilitySwitcherSlot}
  }
  Text_Title {
   widgetClassPath=&CommonTextBlock
   bIsVariable=true
   parent=SizeBox_Title
   widget=Text_Title
   properties {
    autoWrapText=true
    lineHeightPercentage=0.8
    mobileFontSizeMultiplier=0.6
    text="[PH] Title text on three lines. and four"
    textTransformPolicy=ToUpper
    font {
     bForceMonospaced=false
     bMaterialIsStencil=false
     fontMaterial=&M_UI_Billboard_Title
     fontObject=&HeadingNow
     letterSpacing=0
     monospacedWidth=1
     size=75
     skewAmount=0
     typefaceFontName=86 Bold Italic
     outlineSettings {
      bApplyOutlineToDropShadows=false
      bMiteredCorners=false
      bSeparateFillAlpha=false
      outlineMaterial=null
      outlineSize=0
      outlineColor=<rgba 0 0 0 1>
     }
    }
    shadowOffset=<0 0>
   }
   slot {class=SizeBoxSlot, properties {padding=<ltrb 0 -22 0 0>}}
  }
  NamedSlot_TitleImage {
   widgetClassPath=&NamedSlot
   bIsVariable=true
   parent=TitleSwitcher
   widget=NamedSlot_TitleImage
   properties {visibility=Collapsed}
   slot {class=CommonVisibilitySwitcherSlot}
  }
  Text_Subtitle {
   widgetClassPath=&CommonTextBlock
   bIsVariable=true
   parent=VerticalBox_1
   widget=Text_Subtitle
   properties {
    autoWrapText=true
    text="[PH] Subtitle"
    textTransformPolicy=ToUpper
    colorAndOpacity {colorUseRule=UseColor_Specified, specifiedColor=<rgba 0.5776 0.5776 0.5776 1>}
    font {
     bForceMonospaced=false
     bMaterialIsStencil=false
     fontMaterial=null
     fontObject=&HeadingNow
     letterSpacing=0
     monospacedWidth=1
     size=25.5
     skewAmount=0
     typefaceFontName=76 Bold Italic
     outlineSettings {
      bApplyOutlineToDropShadows=false
      bMiteredCorners=false
      bSeparateFillAlpha=false
      outlineMaterial=null
      outlineSize=0
      outlineColor=<rgba 0 0 0 1>
     }
    }
    shadowOffset=<0 0>
   }
   slot {class=VerticalBoxSlot, properties {padding=<ltrb 0 24 0 -9>}}
  }
  Switcher_ItemCards {
   widgetClassPath=&CommonVisibilitySwitcher
   bIsVariable=true
   parent=VerticalBox_All
   widget=Switcher_ItemCards
   slot {class=VerticalBoxSlot, properties {horizontalAlignment=HAlign_Left, padding=<ltrb 0 48 0 0>}}
  }
  SmartCardWrapper {
   widgetClassPath=/ItemShop/MultiProduct/ItemShopTileComponents/WBP_ItemShopTile_SmartCardWrapper
   bIsVariable=true
   parent=Switcher_ItemCards
   widget=SmartCardWrapper
   properties {entryBoxWrapOverride=Horizontal, isUsedInBillboard=true, priority=1, gridSizeOverride=<5 1>}
   slot {class=CommonVisibilitySwitcherSlot}
  }
  NamedSlot_BodyImage {
   widgetClassPath=&NamedSlot
   bIsVariable=true
   parent=Switcher_ItemCards
   widget=NamedSlot_BodyImage
   properties {visibility=Collapsed}
   slot {class=CommonVisibilitySwitcherSlot}
  }
  WBP_Billboard_CTAButtonAndPrice {
   widgetClassPath=&/WBP_Billboard_CTAButtonAndPrice
   bIsVariable=true
   parent=VerticalBox_All
   widget=WBP_Billboard_CTAButtonAndPrice
   properties {layoutSpacing=<40 56>}
   slot {class=VerticalBoxSlot, properties {horizontalAlignment=HAlign_Left, padding=<ltrb 0 48 0 0>}}
  }
 ]
}
