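# Widget-less ticker for item shop billboards that feeds smoothed mouse position into MPC_MousePosition, apparently for UI parallax.
# Tick clamps the viewport mouse position to 1920x1080, spring-interpolates it and writes MousePositionX/Y to the collection.
# UpdateMouseTickerEnabled combines the requested state with the user's bAllowUIParallax setting to toggle visibility, collapsing and resetting to the default centre when off.
# An older Vector2DInterpTo chain remains, disconnected from Tick.
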WidgetBlueprint /ItemShop/MultiProduct/Billboard/WBP_Billboard_MousePositionTicker
graphs [
 EventGraph {nodes [
   n1 = GetMousePositionOnViewport!(execute) -> (then, ReturnValue:Vector 2D Structure)
   n2 = Event Tick()
   n3 = Set MousePos(
    execute=@n1.then
    MousePos:Vector 2D Structure=Vector2DInterpTo!(
     Current:Vector 2D Structure=Get MousePos() -> (MousePos:Vector 2D Structure)
     Target:Vector 2D Structure=@n1
     DeltaTime:float=@n2.InDeltaTime
     InterpSpeed:float=Get Speed() -> (Speed:float)
    ) -> (ReturnValue:Vector 2D Structure)
   ) -> (Output_Get:Vector 2D Structure)
   n4 = BreakVector2D(InVec=@n3)
   > SetScalarParameterValue!(
    execute
    Collection:Material Parameter Collection Object Reference=&/MPC_MousePosition
    ParameterName:Name=MousePositionX
    ParameterValue:float=@n4
   ) -> (then) {is=title K2Node_CallMaterialParameterCollectionFunction}
   > SetScalarParameterValue(Collection=&MPC_MousePosition, ParameterName=MousePositionY, ParameterValue=@n4.Y)
   n5 = Get!() -> (MousePos_X:double, MousePos_Y:double)
   n6 = GetMousePositionOnViewport!(execute=@n2.then) -> (then, ReturnValue_X:double, ReturnValue_Y:double)
   n7 = MakeVector2D!(X="Clamp (Float)"(Value=@n6, Min=0, Max=1920), Y="Clamp (Float)"(Value=@n6.ReturnValue_Y, Min=0, Max=1080)) -> (ReturnValue_X:double, ReturnValue_Y:double)
   > n8 = VectorSpringInterp!(
    execute
    Current:Vector=MakeVector(X=@n5, Y=@n5.MousePos_Y)
    Target:Vector=MakeVector(X=@n7, Y=@n7.ReturnValue_Y)
    SpringState:"Vector Spring State Structure (by ref)"=Get Spring State() -> (Spring State:Vector Spring State Structure)
    Stiffness:float=Get Stiffness() -> (Stiffness:float)
    CriticalDampingFactor:float=Get CriticalDampingFactor() -> (CriticalDampingFactor:float)
    DeltaTime:float=@n2.InDeltaTime
    Mass:float=Get Mass() -> (Mass:float)
    TargetVelocityAmount:float=Get Target Velocity Amount() -> (Target Velocity Amount:float)
    bClamp:bool=false
    MinValue:Vector="-1,-1,-1"
    MaxValue:Vector="1,1,1"
    bInitializeFromTarget:bool=false
   ) -> (then, ReturnValue_X:double, ReturnValue_Y:double, ReturnValue_Z:double)
   > n9 = Set MousePos(MousePos=MakeVector2D!(X=@n8, Y=@n8.ReturnValue_Y) -> (ReturnValue:Vector 2D Structure))
   n10 = BreakVector2D(InVec=@n9)
   SetScalarParameterValue(execute=@n11.then, Collection=&MPC_MousePosition, ParameterName=MousePositionY, ParameterValue=@n10.Y)
   n11 = SetScalarParameterValue(execute=@n9.then, Collection=&MPC_MousePosition, ParameterName=MousePositionX, ParameterValue=@n10)
  ]}
 ResetMousePosition {nodes [
   ResetMousePosition() {is=title K2Node_FunctionEntry}
   > SetScalarParameterValue(
    Collection=&MPC_MousePosition
    ParameterName=MousePositionX
    ParameterValue=Get DefaultMousePositionX() -> (DefaultMousePositionX:double)
   )
   > SetScalarParameterValue(
    Collection=&MPC_MousePosition
    ParameterName=MousePositionY
    ParameterValue=Get DefaultMousePositionY() -> (DefaultMousePositionY:double)
   )
  ]}
 UpdateMouseTickerEnabled {nodes [
   n1 = UpdateMouseTickerEnabled!() -> (then, bDesiredEnabled:bool=~) {is=title K2Node_FunctionEntry}
   n2 = AND Boolean(
    A=@n1
    B=Get bAllowUIParallax!(self:Fort Game User Settings Object Reference=GetFortGameUserSettings!() -> (ReturnValue:Fort Game User Settings Object Reference)) -> (bAllowUIParallax:bool)
   )
   > SetVisibility(InVisibility=Select(Option 0=Collapsed, Option 1=SelfHitTestInvisible, Index=@n2))
   > n3 = Branch(Condition=@n2)
   ResetMousePosition(execute=@n3.else) {is=title K2Node_CallFunction}
  ]}
]
variables [
 CriticalDampingFactor {type=number, default=2}
 DefaultMousePositionX {type=number, default=960}
 DefaultMousePositionY {type=number, default=540}
 Mass {type=number, default=1}
 MousePos {type=Vector2D, default=<0 0>}
 Speed {type=number, default=2}
 Spring State {type=VectorSpringState, default {}}
 Stiffness {type=number, default=0.5}
 Target Velocity Amount {type=number, default=0.8}
]
widgets {info {inheritedWidgetCount=0, parentClass=&UserWidget, rootWidgetClass=null, widgetCount=0}, widgets []}
